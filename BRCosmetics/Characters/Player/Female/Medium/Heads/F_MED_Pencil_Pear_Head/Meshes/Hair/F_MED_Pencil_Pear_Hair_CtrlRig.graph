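# Control Rig for the hair of the female medium Pencil Pear head, correcting dynamic hair bone poses by head pose and character state.
# Parent constraints blend hair strands toward each other, head, neck_01 and spine_05.
# Spherical pose readers on the head weight corrective ModifyTransforms offsets.
# Variables like Sprint, Melee and Moving drive AlphaInterp blends, Sailor and Blazer scale extra correctives, and Skydive, Crouch and ADS branches select per-stance adjustments.
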ControlRigBlueprint /BRCosmetics/Characters/Player/Female/Medium/Heads/F_MED_Pencil_Pear_Head/Meshes/Hair/F_MED_Pencil_Pear_Hair_CtrlRig
graphs [
 RigVMFunctionLibrary {links [], nodes []}
 RigVMModel {nodes [
   &RigUnit_ParentConstraint(
    AdvancedSettings=(RotationOrderForFilter=XZY)
    Child=(Type=Bone,Name="dyn_hair_bk_l_03")
    ExecutePin=()
    Filter=(TranslationFilter=(bX=true,bY=true,bZ=true),RotationFilter=(bX=True,bY=True,bZ=True),ScaleFilter=(bX=false,bY=false,bZ=false))
    Parents=((Item=(Type=Bone,Name="dyn_hair_l_03"),Weight=1))
    Weight=0.3
   )
   > &RigUnit_ParentConstraint(
    AdvancedSettings=(RotationOrderForFilter=XZY)
    Child=(Type=Bone,Name="dyn_hair_bk_l_04")
    Filter=(TranslationFilter=(bX=false,bY=false,bZ=false),RotationFilter=(bX=True,bY=True,bZ=True),ScaleFilter=(bX=false,bY=false,bZ=false))
    Parents=((Item=(Type=Bone,Name="dyn_hair_l_04"),Weight=1))
    Weight=0.3
   )
   RigUnit_BeginExecution = &RigUnit_BeginExecution()
   &RigUnit_ParentConstraint(
    AdvancedSettings=(RotationOrderForFilter=XZY)
    Child=(Type=Bone,Name="dyn_hair_fr_l_03")
    Filter=(TranslationFilter=(bX=false,bY=false,bZ=false),RotationFilter=(bX=True,bY=True,bZ=True),ScaleFilter=(bX=false,bY=false,bZ=false))
    Parents=((Item=(Type=Bone,Name="head"),Weight=0.5))
    Weight=0.2
    ExecutePin=@RigUnit_BeginExecution.ExecutePin
   )
   > &RigUnit_ParentConstraint(
    AdvancedSettings=(RotationOrderForFilter=XZY)
    Child=(Type=Bone,Name="dyn_hair_l_03")
    Filter=(TranslationFilter=(bX=false,bY=false,bZ=false),RotationFilter=(bX=True,bY=True,bZ=True),ScaleFilter=(bX=false,bY=false,bZ=false))
    Parents=((Item=(Type=Bone,Name="dyn_hair_bk_l_02"),Weight=1))
    Weight=0.3
   )
   > &RigUnit_ParentConstraint(
    AdvancedSettings=(RotationOrderForFilter=XZY)
    Child=(Type=Bone,Name="dyn_hair_bk_r_02")
    Filter=(TranslationFilter=(bX=false,bY=false,bZ=false),RotationFilter=(bX=True,bY=True,bZ=True),ScaleFilter=(bX=false,bY=false,bZ=false))
    Parents=((Item=(Type=Bone,Name="dyn_hair_bk_02"),Weight=0.5))
    Weight=0.2
   )
   > &RigUnit_ParentConstraint(
    AdvancedSettings=(RotationOrderForFilter=XZY)
    Child=(Type=Bone,Name="dyn_hair_bk_l_02")
    Filter=(TranslationFilter=(bX=false,bY=false,bZ=false),RotationFilter=(bX=True,bY=True,bZ=True),ScaleFilter=(bX=false,bY=false,bZ=false))
    Parents=((Item=(Type=Bone,Name="dyn_hair_bk_02"),Weight=0.2))
    Weight=0.1
   )
   > &RigUnit_ParentConstraint(
    AdvancedSettings=(RotationOrderForFilter=XZY)
    Child=(Type=Bone,Name="dyn_hair_r_02")
    Filter=(TranslationFilter=(bX=false,bY=false,bZ=false),RotationFilter=(bX=True,bY=True,bZ=True),ScaleFilter=(bX=false,bY=false,bZ=false))
    Parents=((Item=(Type=Bone,Name="dyn_hair_bk_r_02"),Weight=0.5))
    Weight=0.2
   )
   > &RigUnit_ParentConstraint(
    AdvancedSettings=(RotationOrderForFilter=XZY)
    Child=(Type=Bone,Name="dyn_hair_r_04")
    Filter=(TranslationFilter=(bX=false,bY=false,bZ=false),RotationFilter=(bX=True,bY=True,bZ=True),ScaleFilter=(bX=false,bY=false,bZ=false))
    Parents=((Item=(Type=Bone,Name="dyn_hair_bk_r_04"),Weight=0.5))
    Weight=0.2
   )
   > n1 = &RigUnit_SphericalPoseReader(
    ActiveRegionScaleFactors=(PositiveWidth=0.48,NegativeWidth=0.2,PositiveHeight=1,NegativeHeight=1)
    ActiveRegionSize=0.2
    Debug=(bDrawDebug=false,DebugScale=25,DebugSegments=20,DebugThickness=0.25)
    DriverAxis=(X=1,Y=0,Z=0)
    DriverItem=(Type=Bone,Name="head")
    FalloffRegionScaleFactors=(PositiveWidth=0.28,NegativeWidth=0.1,PositiveHeight=0.4,NegativeHeight=0.2)
    FalloffSize=0.5
    OptionalParentItem=(Type=Bone,Name="neck_01")
    RotationOffset=(X=-90,Y=0,Z=55)
   )
   > &RigUnit_ParentConstraint(
    AdvancedSettings=(RotationOrderForFilter=XZY)
    Child=(Type=Bone,Name="dyn_hair_bk_r_02")
    Filter=(TranslationFilter=(bX=false,bY=false,bZ=false),RotationFilter=(bX=True,bY=True,bZ=True),ScaleFilter=(bX=false,bY=false,bZ=false))
    Parents=((Item=(Type=Bone,Name="neck_01"),Weight=0.3))
    Weight=@n1.OutputParam
   )
   > &RigUnit_ModifyTransforms(
    ItemToModify="((Item=(Type=Bone,Name=\"dyn_hair_l_01\"),Transform=(Rotation=(X=0,Y=0,Z=0,W=1),Translation=(X=0,Y=0,Z=0),Scale3D=(X=1,Y=1,Z=1))),(Item=(Type=Bone,Name=\"dyn_hair_bk_l_01\"),Transform=(Rotation=(X=0,Y=0,Z=0,W=1),Translation=(X=0,Y=0,Z=3),Scale3D=(X=1,Y=1,Z=1)) ...(truncated)"
    ItemToModify.1.Transform.Rotation=&RigVMFunction_MathQuaternionFromEuler(Euler=(X=0,Y=-30,Z=0)).Result
    ItemToModify.2.Transform.Rotation=&RigVMFunction_MathQuaternionFromEuler(Euler=(X=0,Y=0,Z=-5)).Result
    ItemToModify.3.Transform.Rotation=&RigVMFunction_MathQuaternionFromEuler(Euler=(X=0,Y=-8,Z=0)).Result
    ItemToModify.0.Transform.Rotation=&RigVMFunction_MathQuaternionFromEuler(Euler=(X=0,Y=-5,Z=-10)).Result
    Weight=@n1.OutputParam
   )
   > &RigUnit_ModifyTransforms(
    ItemToModify=((Item=(Type=Bone,Name="dyn_hair_bk_01"),Transform=(Rotation=(X=0,Y=0,Z=0,W=1),Translation=(X=0,Y=0,Z=5),Scale3D=(X=1,Y=1,Z=1))))
    ItemToModify.0.Transform.Rotation=&RigVMFunction_MathQuaternionFromEuler(Euler=(X=0,Y=-45,Z=0)).Result
    Weight=@n1.OutputParam
   )
   > &RigUnit_ModifyTransforms(
    ItemToModify="((Item=(Type=Bone,Name=\"dyn_hair_r_01\"),Transform=(Rotation=(X=0,Y=0,Z=0,W=1),Translation=(X=0,Y=0,Z=0),Scale3D=(X=1,Y=1,Z=1))),(Item=(Type=Bone,Name=\"dyn_hair_bk_r_01\"),Transform=(Rotation=(X=0,Y=0,Z=0,W=1),Translation=(X=-4,Y=2,Z=3),Scale3D=(X=1,Y=1,Z=1) ...(truncated)"
    ItemToModify.1.Transform.Rotation=&RigVMFunction_MathQuaternionFromEuler(Euler=(X=0,Y=0,Z=0)).Result
    ItemToModify.2.Transform.Rotation=&RigVMFunction_MathQuaternionFromEuler(Euler=(X=0,Y=-15,Z=0)).Result
    ItemToModify.3.Transform.Rotation=&RigVMFunction_MathQuaternionFromEuler(Euler=(X=0,Y=-60,Z=60)).Result
    Weight=@n1.OutputParam
   )
   n2 = &RigVMFunction_AlphaInterp(
    InterpSpeedDecreasing=10
    InterpSpeedIncreasing=10
    bClampResult=False
    bInterpResult=False
    Value=@n1.OutputParam
    Scale=RigVMVariableNode() {variable=Sailor}.Value
   )
   > &RigUnit_ModifyTransforms(
    ItemToModify="((Item=(Type=Bone,Name=\"dyn_hair_fr_l_02\"),Transform=(Rotation=(X=0,Y=0,Z=0,W=1),Translation=(X=0,Y=0,Z=0),Scale3D=(X=1,Y=1,Z=1))),(Item=(Type=Bone,Name=\"dyn_hair_bk_01\"),Transform=(Rotation=(X=0,Y=0,Z=0,W=1),Translation=(X=0,Y=0,Z=0),Scale3D=(X=1,Y=1,Z=1) ...(truncated)"
    Weight=@n2.Result
    ItemToModify.3.Transform.Rotation=&RigVMFunction_MathQuaternionFromEuler(Euler=(X=0,Y=0,Z=0), RotationOrder=XYZ).Result
   )
   n3 = &RigVMFunction_AlphaInterp(InterpSpeedDecreasing=10, InterpSpeedIncreasing=10, Scale=1, bClampResult=False, Value=RigVMVariableNode() {variable=Rifle}.Value)
   > &RigUnit_ModifyTransforms(
    ItemToModify="((Item=(Type=Bone,Name=\"dyn_hair_bk_r_01\"),Transform=(Rotation=(X=0,Y=0,Z=0,W=1),Translation=(X=0,Y=0,Z=0),Scale3D=(X=1,Y=1,Z=1))),(Item=(Type=Bone,Name=\"dyn_hair_r_01\"),Transform=(Rotation=(X=0,Y=0,Z=0,W=1),Translation=(X=0,Y=0,Z=0),Scale3D=(X=1,Y=1,Z=1)) ...(truncated)"
    Weight=@n3.Result
    ItemToModify.1.Transform.Rotation=&RigVMFunction_MathQuaternionFromEuler(Euler=(X=0,Y=-30,Z=0)).Result
    ItemToModify.0.Transform.Rotation=&RigVMFunction_MathQuaternionFromEuler(Euler=(X=0,Y=0,Z=0)).Result
   )
   > n4 = &RigUnit_SphericalPoseReader(
    ActiveRegionScaleFactors=(PositiveWidth=0.1,NegativeWidth=0.8,PositiveHeight=1,NegativeHeight=1)
    ActiveRegionSize=0.2
    Debug=(bDrawDebug=false,DebugScale=25,DebugSegments=20,DebugThickness=0.25)
    DriverAxis=(X=1,Y=0,Z=0)
    DriverItem=(Type=Bone,Name="head")
    FalloffRegionScaleFactors=(PositiveWidth=0.2,NegativeWidth=0.3,PositiveHeight=0.5,NegativeHeight=0.5)
    FalloffSize=0.45
    OptionalParentItem=(Type=Bone,Name="neck_01")
    RotationOffset=(X=-90,Y=0,Z=130)
   )
   > &RigUnit_ParentConstraint(
    AdvancedSettings=(RotationOrderForFilter=XZY)
    Child=(Type=Bone,Name="dyn_hair_bk_r_02")
    Filter=(TranslationFilter=(bX=false,bY=false,bZ=false),RotationFilter=(bX=True,bY=True,bZ=True),ScaleFilter=(bX=false,bY=false,bZ=false))
    Parents=((Item=(Type=Bone,Name="neck_01"),Weight=0.3))
    Weight=@n4.OutputParam
   )
   > &RigUnit_ModifyTransforms(
    ItemToModify="((Item=(Type=Bone,Name=\"dyn_hair_l_01\"),Transform=(Rotation=(X=0,Y=0,Z=0,W=1),Translation=(X=0,Y=0,Z=0),Scale3D=(X=1,Y=1,Z=1))),(Item=(Type=Bone,Name=\"dyn_hair_fr_l_02\"),Transform=(Rotation=(X=0,Y=0,Z=0,W=1),Translation=(X=0,Y=0,Z=0),Scale3D=(X=1,Y=1,Z=1)) ...(truncated)"
    ItemToModify.0.Transform.Rotation=&RigVMFunction_MathQuaternionFromEuler(Euler=(X=0,Y=-2,Z=20)).Result
    ItemToModify.1.Transform.Rotation=&RigVMFunction_MathQuaternionFromEuler(Euler=(X=0,Y=0,Z=0)).Result
    Weight=@n4.OutputParam
   )
   > &RigUnit_ModifyTransforms(
    ItemToModify="((Item=(Type=Bone,Name=\"dyn_hair_bk_01\"),Transform=(Rotation=(X=0,Y=0,Z=0,W=1),Translation=(X=0,Y=0,Z=0),Scale3D=(X=1,Y=1,Z=1))),(Item=(Type=Bone,Name=\"dyn_hair_bk_l_02\"),Transform=(Rotation=(X=0,Y=0,Z=0,W=1),Translation=(X=0,Y=0,Z=0),Scale3D=(X=1,Y=1,Z=1) ...(truncated)"
    ItemToModify.1.Transform.Rotation=&RigVMFunction_MathQuaternionFromEuler(Euler=(X=0,Y=-10,Z=0)).Result
    ItemToModify.2.Transform.Rotation=&RigVMFunction_MathQuaternionFromEuler(Euler=(X=0,Y=-10,Z=0)).Result
    ItemToModify.0.Transform.Rotation=&RigVMFunction_MathQuaternionFromEuler(Euler=(X=0,Y=0,Z=-5)).Result
    Weight=@n4.OutputParam
   )
   AlphaInterp = &RigVMFunction_AlphaInterp(InterpSpeedDecreasing=7, InterpSpeedIncreasing=3, Scale=1, bClampResult=False, Value=RigVMVariableNode() {variable=Melee}.Value)
   > &RigUnit_ParentConstraint(
    AdvancedSettings=(RotationOrderForFilter=XZY)
    Child=(Type=Bone,Name="dyn_hair_bk_l_03")
    Parents=((Item=(Type=Bone,Name="spine_05"),Weight=1))
    Weight=@AlphaInterp.Result
   )
   > &RigUnit_ParentConstraint(
    AdvancedSettings=(RotationOrderForFilter=XZY)
    Child=(Type=Bone,Name="dyn_hair_bk_02")
    Parents=((Item=(Type=Bone,Name="spine_05"),Weight=1))
    Parents.0.Weight=@AlphaInterp.Result
    Weight=@AlphaInterp.Result
   )
   > &RigUnit_ParentConstraint(
    AdvancedSettings=(RotationOrderForFilter=XZY)
    Child=(Type=Bone,Name="dyn_hair_bk_r_03")
    Parents=((Item=(Type=Bone,Name="spine_05"),Weight=1))
    Parents.0.Weight=@AlphaInterp.Result
    Weight=@AlphaInterp.Result
   )
   > &RigUnit_ParentConstraint(
    AdvancedSettings=(RotationOrderForFilter=XZY)
    Child=(Type=Bone,Name="dyn_hair_fr_l_02")
    Parents=((Item=(Type=Bone,Name="spine_05"),Weight=1))
    Weight=@AlphaInterp.Result
   )
   > &RigUnit_ParentConstraint(
    AdvancedSettings=(RotationOrderForFilter=XZY)
    Child=(Type=Bone,Name="dyn_hair_fr_r_02")
    Parents=((Item=(Type=Bone,Name="spine_05"),Weight=1))
    Weight=@AlphaInterp.Result
   )
   > &RigUnit_ParentConstraint(
    AdvancedSettings=(RotationOrderForFilter=XZY)
    Child=(Type=Bone,Name="dyn_hair_r_03")
    Parents=((Item=(Type=Bone,Name="spine_05"),Weight=1))
    Weight=@AlphaInterp.Result
   )
   > n5 = &RigUnit_ParentConstraint(
    AdvancedSettings=(RotationOrderForFilter=XZY)
    Child=(Type=Bone,Name="dyn_hair_l_03")
    Parents=((Item=(Type=Bone,Name="spine_05"),Weight=1))
    Weight=@AlphaInterp.Result
   )
   Branch_1_RigVMUnitNode = &RigVMFunction_ControlFlowBranch(ExecuteContext=@n5.ExecutePin, Condition=RigVMVariableNode() {variable=Skydive}.Value)
   Branch_RigVMUnitNode = &RigVMFunction_ControlFlowBranch(ExecuteContext=@Branch_1_RigVMUnitNode.False, Condition=RigVMVariableNode() {variable=Crouch}.Value)
   Branch_1_1_RigVMUnitNode = &RigVMFunction_ControlFlowBranch(ExecuteContext=@Branch_RigVMUnitNode.False, Condition=RigVMVariableNode() {variable=ADS_Rifle}.Value)
   Branch_1_1_1_2_RigVMUnitNode = &RigVMFunction_ControlFlowBranch(ExecuteContext=@Branch_1_1_RigVMUnitNode.False, Condition=RigVMVariableNode() {variable=ADS_Pistol}.Value)
   n6 = &RigUnit_SphericalPoseReader(
    ActiveRegionScaleFactors=(PositiveWidth=0.2,NegativeWidth=0.1,PositiveHeight=1,NegativeHeight=1)
    ActiveRegionSize=0.2
    Debug=(bDrawDebug=false,DebugScale=25,DebugSegments=20,DebugThickness=0.25)
    DriverAxis=(X=1,Y=0,Z=0)
    DriverItem=(Type=Bone,Name="head")
    FalloffRegionScaleFactors=(PositiveWidth=0.1,NegativeWidth=0.1,PositiveHeight=0.5,NegativeHeight=0.5)
    FalloffSize=0.5
    OptionalParentItem=(Type=Bone,Name="neck_01")
    RotationOffset=(X=-90,Y=0,Z=90)
    ExecutePin=@Branch_1_1_1_2_RigVMUnitNode.False
   )
   > &RigUnit_ModifyTransforms(
    ItemToModify=((Item=(Type=Bone,Name="dyn_hair_r_02"),Transform=(Rotation=(X=0,Y=0,Z=0,W=1),Translation=(X=0,Y=0,Z=0),Scale3D=(X=1,Y=1,Z=1))))
    ItemToModify.0.Transform.Rotation=&RigVMFunction_MathQuaternionFromEuler(Euler=(X=0,Y=0,Z=0)).Result
    Weight=@n6.OutputParam
   )
   > n7 = &RigUnit_SphericalPoseReader(
    ActiveRegionScaleFactors=(PositiveWidth=0.2,NegativeWidth=0.1,PositiveHeight=1,NegativeHeight=1)
    ActiveRegionSize=0.2
    Debug=(bDrawDebug=false,DebugScale=25,DebugSegments=20,DebugThickness=0.25)
    DriverAxis=(X=1,Y=0,Z=0)
    DriverItem=(Type=Bone,Name="head")
    FalloffRegionScaleFactors=(PositiveWidth=0.1,NegativeWidth=0.1,PositiveHeight=0.5,NegativeHeight=0.5)
    FalloffSize=0.5
    OptionalParentItem=(Type=Bone,Name="neck_01")
    RotationOffset=(X=-90,Y=0,Z=60)
   )
   > &RigUnit_ModifyTransforms(
    ItemToModify=((Item=(Type=Bone,Name="dyn_hair_r_01"),Transform=(Rotation=(X=0,Y=0,Z=0,W=1),Translation=(X=0,Y=0,Z=0),Scale3D=(X=1,Y=1,Z=1))))
    ItemToModify.0.Transform.Rotation=&RigVMFunction_MathQuaternionFromEuler(Euler=(X=0,Y=-30,Z=0)).Result
    Weight=@n7.OutputParam
   )
   n8 = &RigVMFunction_AlphaInterp(InterpSpeedDecreasing=7, InterpSpeedIncreasing=2, Scale=1, bClampResult=False, Value=RigVMVariableNode() {variable=Sprint}.Value)
   > &RigUnit_ModifyTransforms(
    ItemToModify="((Item=(Type=Bone,Name=\"dyn_hair_bk_l_01\"),Transform=(Rotation=(X=0,Y=0,Z=0,W=1),Translation=(X=0,Y=0,Z=0),Scale3D=(X=1,Y=1,Z=1))),(Item=(Type=Bone,Name=\"dyn_hair_bk_01\"),Transform=(Rotation=(X=0,Y=0,Z=0,W=1),Translation=(X=0,Y=0,Z=0),Scale3D=(X=1,Y=1,Z=1) ...(truncated)"
    Weight=@n8.Result
    ItemToModify.2.Transform.Rotation=&RigVMFunction_MathQuaternionFromEuler(Euler=(X=0,Y=-30,Z=0)).Result
    ItemToModify.3.Transform.Rotation=&RigVMFunction_MathQuaternionFromEuler(Euler=(X=0,Y=-30,Z=0)).Result
    ItemToModify.4.Transform.Rotation=&RigVMFunction_MathQuaternionFromEuler(Euler=(X=0,Y=-60,Z=0)).Result
    ItemToModify.1.Transform.Rotation=&RigVMFunction_MathQuaternionFromEuler(Euler=(X=0,Y=-60,Z=0)).Result
    ItemToModify.0.Transform.Rotation=&RigVMFunction_MathQuaternionFromEuler(Euler=(X=0,Y=-30,Z=0)).Result
   )
   > &RigUnit_ModifyTransforms(
    ItemToModify="((Item=(Type=Bone,Name=\"dyn_hair_bk_01\"),Transform=(Rotation=(X=0,Y=0,Z=0,W=1),Translation=(X=0,Y=0,Z=0),Scale3D=(X=1,Y=1,Z=1))),(Item=(Type=Bone,Name=\"dyn_hair_bk_l_01\"),Transform=(Rotation=(X=0,Y=0,Z=0,W=1),Translation=(X=0,Y=0,Z=0),Scale3D=(X=1,Y=1,Z=1) ...(truncated)"
    ItemToModify.0.Transform.Rotation=&RigVMFunction_MathQuaternionFromEuler(Euler=(X=0,Y=-30,Z=0)).Result
    ItemToModify.1.Transform.Rotation=&RigVMFunction_MathQuaternionFromEuler(Euler=(X=0,Y=-45,Z=0)).Result
    ItemToModify.2.Transform.Rotation=&RigVMFunction_MathQuaternionFromEuler(Euler=(X=0,Y=-20,Z=0)).Result
    ItemToModify.3.Transform.Rotation=&RigVMFunction_MathQuaternionFromEuler(Euler=(X=0,Y=-45,Z=0)).Result
    ItemToModify.4.Transform.Rotation=&RigVMFunction_MathQuaternionFromEuler(Euler=(X=0,Y=0,Z=-60)).Result
    ItemToModify.5.Transform.Rotation=&RigVMFunction_MathQuaternionFromEuler(Euler=(X=0,Y=-125,Z=0)).Result
    Weight=@n8.Result
   )
   &RigUnit_ParentConstraint(
    AdvancedSettings=(RotationOrderForFilter=XZY)
    Child=(Type=Bone,Name="dyn_hair_bk_02")
    Filter=(TranslationFilter=(bX=True,bY=True,bZ=True),RotationFilter=(bX=True,bY=True,bZ=True),ScaleFilter=(bX=false,bY=false,bZ=false))
    Parents=((Item=(Type=Bone,Name="spine_05"),Weight=1))
    Weight=1
    ExecutePin=@Branch_1_1_1_2_RigVMUnitNode.True
   )
   > &RigUnit_ParentConstraint(
    AdvancedSettings=(RotationOrderForFilter=XZY)
    Child=(Type=Bone,Name="dyn_hair_bk_r_03")
    Filter=(TranslationFilter=(bX=True,bY=True,bZ=True),RotationFilter=(bX=True,bY=True,bZ=True),ScaleFilter=(bX=false,bY=false,bZ=false))
    Parents=((Item=(Type=Bone,Name="spine_05"),Weight=1))
    Weight=1
   )
   > &RigUnit_ParentConstraint(
    AdvancedSettings=(RotationOrderForFilter=XZY)
    Child=(Type=Bone,Name="dyn_hair_bk_l_03")
    Filter=(TranslationFilter=(bX=True,bY=True,bZ=True),RotationFilter=(bX=True,bY=True,bZ=True),ScaleFilter=(bX=false,bY=false,bZ=false))
    Parents=((Item=(Type=Bone,Name="spine_05"),Weight=1))
    Weight=1
   )
   > &RigUnit_ParentConstraint(
    AdvancedSettings=(RotationOrderForFilter=XZY)
    Child=(Type=Bone,Name="dyn_hair_r_03")
    Filter=(TranslationFilter=(bX=true,bY=true,bZ=true),RotationFilter=(bX=True,bY=True,bZ=True),ScaleFilter=(bX=false,bY=false,bZ=false))
    Parents=((Item=(Type=Bone,Name="spine_05"),Weight=1))
    Weight=1
   )
   > n9 = &RigUnit_SphericalPoseReader(
    ActiveRegionScaleFactors=(PositiveWidth=0.5,NegativeWidth=0.2,PositiveHeight=0.5,NegativeHeight=0.5)
    ActiveRegionSize=0.15
    Debug=(bDrawDebug=false,DebugScale=25,DebugSegments=20,DebugThickness=0.25)
    DriverAxis=(X=1,Y=0,Z=0)
    DriverItem=(Type=Bone,Name="head")
    FalloffRegionScaleFactors=(PositiveWidth=0.3,NegativeWidth=0.05,PositiveHeight=0.5,NegativeHeight=0.5)
    FalloffSize=0.4
    OptionalParentItem=(Type=Bone,Name="None")
    RotationOffset=(X=100,Y=0,Z=-70)
   )
   > &RigUnit_ModifyTransforms(
    ItemToModify="((Item=(Type=Bone,Name=\"dyn_hair_fr_l_02\"),Transform=(Rotation=(X=0,Y=0,Z=0,W=1),Translation=(X=0,Y=0,Z=0),Scale3D=(X=1,Y=1,Z=1))),(Item=(Type=Bone,Name=\"dyn_hair_bk_01\"),Transform=(Rotation=(X=0,Y=0,Z=0,W=1),Translation=(X=0,Y=0,Z=0),Scale3D=(X=1,Y=1,Z=1) ...(truncated)"
    ItemToModify.2.Transform.Rotation=&RigVMFunction_MathQuaternionFromEuler(Euler=(X=0,Y=0,Z=0), RotationOrder=XYZ).Result
    ItemToModify.4.Transform.Rotation=&RigVMFunction_MathQuaternionFromEuler(Euler=(X=0,Y=0,Z=-30), RotationOrder=XYZ).Result
    ItemToModify.3.Transform.Rotation=&RigVMFunction_MathQuaternionFromEuler(Euler=(X=0,Y=0,Z=-10), RotationOrder=XYZ).Result
    ItemToModify.1.Transform.Rotation=&RigVMFunction_MathQuaternionFromEuler(Euler=(X=0,Y=0,Z=0), RotationOrder=XYZ).Result
    ItemToModify.0.Transform.Rotation=&RigVMFunction_MathQuaternionFromEuler(Euler=(X=0,Y=0,Z=0), RotationOrder=XYZ).Result
    Weight=@n9.OutputParam
   )
   > n10 = &RigUnit_SphericalPoseReader(
    ActiveRegionScaleFactors=(PositiveWidth=0.5,NegativeWidth=0.5,PositiveHeight=0.5,NegativeHeight=0.5)
    ActiveRegionSize=0.15
    Debug=(bDrawDebug=false,DebugScale=25,DebugSegments=20,DebugThickness=0.25)
    DriverAxis=(X=1,Y=0,Z=0)
    DriverItem=(Type=Bone,Name="head")
    FalloffRegionScaleFactors=(PositiveWidth=0.05,NegativeWidth=0.25,PositiveHeight=0.5,NegativeHeight=0.5)
    FalloffSize=0.4
    OptionalParentItem=(Type=Bone,Name="None")
    RotationOffset=(X=75,Y=0,Z=-130)
   )
   > &RigUnit_ModifyTransforms(
    ItemToModify="((Item=(Type=Bone,Name=\"dyn_hair_fr_l_02\"),Transform=(Rotation=(X=0,Y=0,Z=0,W=1),Translation=(X=0,Y=0,Z=0),Scale3D=(X=1,Y=1,Z=1))),(Item=(Type=Bone,Name=\"dyn_hair_bk_01\"),Transform=(Rotation=(X=0,Y=0,Z=0,W=1),Translation=(X=0,Y=0,Z=0),Scale3D=(X=1,Y=1,Z=1) ...(truncated)"
    ItemToModify.2.Transform.Rotation=&RigVMFunction_MathQuaternionFromEuler(Euler=(X=0,Y=-20,Z=0), RotationOrder=XYZ).Result
    ItemToModify.4.Transform.Rotation=&RigVMFunction_MathQuaternionFromEuler(Euler=(X=0,Y=0,Z=0), RotationOrder=XYZ).Result
    ItemToModify.1.Transform.Rotation=&RigVMFunction_MathQuaternionFromEuler(Euler=(X=0,Y=0,Z=-30), RotationOrder=XYZ).Result
    ItemToModify.3.Transform.Rotation=&RigVMFunction_MathQuaternionFromEuler(Euler=(X=0,Y=-20,Z=-30), RotationOrder=XYZ).Result
    Weight=@n10.OutputParam
   )
   n11 = &RigVMFunction_AlphaInterp(
    InterpSpeedDecreasing=10
    InterpSpeedIncreasing=10
    bClampResult=False
    bInterpResult=False
    Value=@n10.OutputParam
    Scale=RigVMVariableNode() {variable=Sailor}.Value
   )
   > &RigUnit_ModifyTransforms(
    ItemToModify="((Item=(Type=Bone,Name=\"dyn_hair_fr_l_02\"),Transform=(Rotation=(X=0,Y=0,Z=0,W=1),Translation=(X=0,Y=0,Z=0),Scale3D=(X=1,Y=1,Z=1))),(Item=(Type=Bone,Name=\"dyn_hair_bk_01\"),Transform=(Rotation=(X=0,Y=0,Z=0,W=1),Translation=(X=0,Y=0,Z=0),Scale3D=(X=0.8,Y=1,Z= ...(truncated)"
    Weight=@n11.Result
    ItemToModify.3.Transform.Rotation=&RigVMFunction_MathQuaternionFromEuler(Euler=(X=0,Y=0,Z=0), RotationOrder=XYZ).Result
   )
   &RigUnit_ParentConstraint(
    AdvancedSettings=(RotationOrderForFilter=XZY)
    Child=(Type=Bone,Name="dyn_hair_bk_02")
    Filter=(TranslationFilter=(bX=True,bY=True,bZ=True),RotationFilter=(bX=True,bY=True,bZ=True),ScaleFilter=(bX=false,bY=false,bZ=false))
    Parents=((Item=(Type=Bone,Name="spine_05"),Weight=1))
    Weight=1
    ExecutePin=@Branch_1_1_RigVMUnitNode.True
   )
   > &RigUnit_ParentConstraint(
    AdvancedSettings=(RotationOrderForFilter=XZY)
    Child=(Type=Bone,Name="dyn_hair_bk_r_03")
    Filter=(TranslationFilter=(bX=True,bY=True,bZ=True),RotationFilter=(bX=True,bY=True,bZ=True),ScaleFilter=(bX=false,bY=false,bZ=false))
    Parents=((Item=(Type=Bone,Name="spine_05"),Weight=1))
    Weight=1
   )
   > &RigUnit_ParentConstraint(
    AdvancedSettings=(RotationOrderForFilter=XZY)
    Child=(Type=Bone,Name="dyn_hair_bk_l_03")
    Filter=(TranslationFilter=(bX=True,bY=True,bZ=True),RotationFilter=(bX=True,bY=True,bZ=True),ScaleFilter=(bX=false,bY=false,bZ=false))
    Parents=((Item=(Type=Bone,Name="spine_05"),Weight=1))
    Weight=1
   )
   > &RigUnit_ParentConstraint(
    AdvancedSettings=(RotationOrderForFilter=XZY)
    Child=(Type=Bone,Name="dyn_hair_r_03")
    Filter=(TranslationFilter=(bX=true,bY=true,bZ=true),RotationFilter=(bX=True,bY=True,bZ=True),ScaleFilter=(bX=false,bY=false,bZ=false))
    Parents=((Item=(Type=Bone,Name="spine_05"),Weight=1))
    Weight=1
   )
   > n12 = &RigUnit_SphericalPoseReader(
    ActiveRegionScaleFactors=(PositiveWidth=0.1,NegativeWidth=0.2,PositiveHeight=0.8,NegativeHeight=0.8)
    ActiveRegionSize=0.15
    Debug=(bDrawDebug=true,DebugScale=25,DebugSegments=20,DebugThickness=0.25)
    DriverAxis=(X=1,Y=0,Z=0)
    DriverItem=(Type=Bone,Name="head")
    FalloffRegionScaleFactors=(PositiveWidth=0.05,NegativeWidth=0.3,PositiveHeight=0.5,NegativeHeight=0.5)
    FalloffSize=0.4
    OptionalParentItem=(Type=Bone,Name="neck_01")
    RotationOffset=(X=80,Y=0,Z=-130)
   )
   > &RigUnit_ParentConstraint(
    AdvancedSettings=(RotationOrderForFilter=XZY)
    Child=(Type=Bone,Name="dyn_hair_fr_l_05")
    Filter=(TranslationFilter=(bX=true,bY=true,bZ=false),RotationFilter=(bX=True,bY=True,bZ=True),ScaleFilter=(bX=false,bY=false,bZ=false))
    Parents=((Item=(Type=Bone,Name="neck_01"),Weight=0.05))
    Weight=@n12.OutputParam
   )
   > &RigUnit_ModifyTransforms(
    ItemToModify="((Item=(Type=Bone,Name=\"dyn_hair_fr_l_02\"),Transform=(Rotation=(X=0,Y=0,Z=0,W=1),Translation=(X=0,Y=0,Z=0),Scale3D=(X=1,Y=1,Z=1))),(Item=(Type=Bone,Name=\"dyn_hair_bk_r_01\"),Transform=(Rotation=(X=0,Y=0,Z=0,W=1),Translation=(X=0,Y=0,Z=3),Scale3D=(X=1,Y=1,Z= ...(truncated)"
    ItemToModify.2.Transform.Rotation=&RigVMFunction_MathQuaternionFromEuler(Euler=(X=0,Y=0,Z=20), RotationOrder=XYZ).Result
    ItemToModify.3.Transform.Rotation=&RigVMFunction_MathQuaternionFromEuler(Euler=(X=0,Y=90,Z=10), RotationOrder=XYZ).Result
    ItemToModify.4.Transform.Rotation=&RigVMFunction_MathQuaternionFromEuler(Euler=(X=0,Y=0,Z=0), RotationOrder=XYZ).Result
    ItemToModify.1.Transform.Rotation=&RigVMFunction_MathQuaternionFromEuler(Euler=(X=0,Y=-20,Z=0), RotationOrder=XYZ).Result
    Weight=@n12.OutputParam
   )
   > n13 = &RigUnit_SphericalPoseReader(
    ActiveRegionScaleFactors=(PositiveWidth=0.5,NegativeWidth=0.5,PositiveHeight=0.5,NegativeHeight=0.5)
    ActiveRegionSize=0.15
    Debug=(bDrawDebug=true,DebugScale=25,DebugSegments=20,DebugThickness=0.25)
    DriverAxis=(X=1,Y=0,Z=0)
    DriverItem=(Type=Bone,Name="head")
    FalloffRegionScaleFactors=(PositiveWidth=0.1,NegativeWidth=0.05,PositiveHeight=0.5,NegativeHeight=0.5)
    FalloffSize=0.4
    OptionalParentItem=(Type=Bone,Name="None")
    RotationOffset=(X=100,Y=0,Z=-75)
   )
   > &RigUnit_ModifyTransforms(
    ItemToModify="((Item=(Type=Bone,Name=\"dyn_hair_fr_l_02\"),Transform=(Rotation=(X=0,Y=0,Z=0,W=1),Translation=(X=0,Y=0,Z=0),Scale3D=(X=1,Y=1,Z=1))),(Item=(Type=Bone,Name=\"dyn_hair_bk_01\"),Transform=(Rotation=(X=0,Y=0,Z=0,W=1),Translation=(X=0,Y=0,Z=0),Scale3D=(X=1,Y=1,Z=1) ...(truncated)"
    ItemToModify.4.Transform.Rotation=&RigVMFunction_MathQuaternionFromEuler(Euler=(X=0,Y=0,Z=0), RotationOrder=XYZ).Result
    ItemToModify.3.Transform.Rotation=&RigVMFunction_MathQuaternionFromEuler(Euler=(X=0,Y=25,Z=0), RotationOrder=XYZ).Result
    ItemToModify.2.Transform.Rotation=&RigVMFunction_MathQuaternionFromEuler(Euler=(X=0,Y=0,Z=0), RotationOrder=XYZ).Result
    ItemToModify.1.Transform.Rotation=&RigVMFunction_MathQuaternionFromEuler(Euler=(X=0,Y=0,Z=5), RotationOrder=XYZ).Result
    Weight=@n13.OutputParam
   )
   n14 = &RigVMFunction_AlphaInterp(
    InterpSpeedDecreasing=10
    InterpSpeedIncreasing=10
    bClampResult=False
    bInterpResult=False
    Value=@n12.OutputParam
    Scale=RigVMVariableNode() {variable=Sailor}.Value
   )
   > &RigUnit_ModifyTransforms(
    ItemToModify="((Item=(Type=Bone,Name=\"dyn_hair_fr_l_02\"),Transform=(Rotation=(X=0,Y=0,Z=0,W=1),Translation=(X=0,Y=0,Z=0),Scale3D=(X=1,Y=1,Z=1))),(Item=(Type=Bone,Name=\"dyn_hair_bk_01\"),Transform=(Rotation=(X=0,Y=0,Z=0,W=1),Translation=(X=0,Y=0,Z=0),Scale3D=(X=1,Y=1,Z=1) ...(truncated)"
    Weight=@n14.Result
    ItemToModify.2.Transform.Rotation=&RigVMFunction_MathQuaternionFromEuler(Euler=(X=0,Y=0,Z=20), RotationOrder=XYZ).Result
    ItemToModify.3.Transform.Rotation=&RigVMFunction_MathQuaternionFromEuler(Euler=(X=0,Y=0,Z=0), RotationOrder=XYZ).Result
   )
   n15 = &RigVMFunction_AlphaInterp(
    InterpSpeedDecreasing=10
    InterpSpeedIncreasing=10
    bClampResult=False
    bInterpResult=False
    Value=@n12.OutputParam
    Scale=RigVMVariableNode() {variable=Blazer}.Value
   )
   > &RigUnit_ModifyTransforms(
    ItemToModify="((Item=(Type=Bone,Name=\"dyn_hair_fr_r_04\"),Transform=(Rotation=(X=0,Y=0,Z=0,W=1),Translation=(X=0,Y=0,Z=0),Scale3D=(X=1,Y=1,Z=1))),(Item=(Type=Bone,Name=\"dyn_hair_fr_r_04\"),Transform=(Rotation=(X=0,Y=0,Z=0,W=1),Translation=(X=0,Y=0,Z=0),Scale3D=(X=1,Y=1,Z= ...(truncated)"
    Weight=@n15.Result
    ItemToModify.1.Transform.Rotation=&RigVMFunction_MathQuaternionFromEuler(Euler=(X=0,Y=0,Z=0)).Result
    ItemToModify.2.Transform.Rotation=&RigVMFunction_MathQuaternionFromEuler(Euler=(X=0,Y=60,Z=0)).Result
    ItemToModify.0.Transform.Rotation=&RigVMFunction_MathQuaternionFromEuler(Euler=(X=0,Y=0,Z=0)).Result
   )
   n16 = &RigVMFunction_AlphaInterp(
    InterpSpeedDecreasing=10
    InterpSpeedIncreasing=10
    Scale=1
    bClampResult=False
    bInterpResult=False
    Value=RigVMVariableNode() {variable=Blazer}.Value
   )
   > &RigUnit_ModifyTransforms(
    ItemToModify="((Item=(Type=Bone,Name=\"dyn_hair_fr_l_04\"),Transform=(Rotation=(X=0,Y=0,Z=0,W=1),Translation=(X=0,Y=0,Z=0),Scale3D=(X=1,Y=1,Z=1))),(Item=(Type=Bone,Name=\"dyn_hair_fr_r_04\"),Transform=(Rotation=(X=0,Y=0,Z=0,W=1),Translation=(X=0,Y=0,Z=0),Scale3D=(X=1,Y=1,Z= ...(truncated)"
    Weight=@n16.Result
    ItemToModify.0.Transform.Rotation=&RigVMFunction_MathQuaternionFromEuler(Euler=(X=0,Y=0,Z=0)).Result
   )
   n17 = &RigVMFunction_AlphaInterp(InterpSpeedDecreasing=10, InterpSpeedIncreasing=5, Scale=1, bClampResult=False, Value=RigVMVariableNode() {variable=Moving}.Value)
   &RigUnit_ParentConstraint(
    AdvancedSettings=(RotationOrderForFilter=XZY)
    Child=(Type=Bone,Name="dyn_hair_bk_r_03")
    Filter=(TranslationFilter=(bX=true,bY=true,bZ=true),RotationFilter=(bX=True,bY=True,bZ=True),ScaleFilter=(bX=false,bY=false,bZ=false))
    Parents=((Item=(Type=Bone,Name="neck_01"),Weight=1))
    Weight=@n17.Result
    ExecutePin=@Branch_RigVMUnitNode.True
   )
   > &RigUnit_ParentConstraint(
    AdvancedSettings=(RotationOrderForFilter=XZY)
    Child=(Type=Bone,Name="dyn_hair_bk_02")
    Filter=(TranslationFilter=(bX=true,bY=true,bZ=true),RotationFilter=(bX=True,bY=True,bZ=True),ScaleFilter=(bX=false,bY=false,bZ=false))
    Parents=((Item=(Type=Bone,Name="spine_05"),Weight=1))
    Weight=@n17.Result
   )
   > &RigUnit_ParentConstraint(
    AdvancedSettings=(RotationOrderForFilter=XZY)
    Child=(Type=Bone,Name="dyn_hair_r_01")
    Filter=(TranslationFilter=(bX=false,bY=false,bZ=false),RotationFilter=(bX=True,bY=True,bZ=True),ScaleFilter=(bX=false,bY=false,bZ=false))
    Parents=((Item=(Type=Bone,Name="spine_05"),Weight=1))
    Weight=@n17.Result
   )
   > &RigUnit_ParentConstraint(
    AdvancedSettings=(RotationOrderForFilter=XZY)
    Child=(Type=Bone,Name="dyn_hair_bk_l_03")
    Filter=(TranslationFilter=(bX=false,bY=false,bZ=false),RotationFilter=(bX=True,bY=True,bZ=True),ScaleFilter=(bX=false,bY=false,bZ=false))
    Parents=((Item=(Type=Bone,Name="spine_05"),Weight=1))
    Weight=@n17.Result
   )
   > &RigUnit_ParentConstraint(
    AdvancedSettings=(RotationOrderForFilter=XZY)
    Child=(Type=Bone,Name="dyn_hair_l_03")
    Filter=(TranslationFilter=(bX=false,bY=false,bZ=false),RotationFilter=(bX=True,bY=True,bZ=True),ScaleFilter=(bX=false,bY=false,bZ=false))
    Parents=((Item=(Type=Bone,Name="neck_01"),Weight=1))
    Weight=@n17.Result
   )
   > &RigUnit_ModifyTransforms(
    ItemToModify="((Item=(Type=Bone,Name=\"dyn_hair_l_01\"),Transform=(Rotation=(X=0,Y=0,Z=0,W=1),Translation=(X=-5,Y=-2,Z=0),Scale3D=(X=1,Y=1,Z=1))),(Item=(Type=Bone,Name=\"dyn_hair_bk_l_01\"),Transform=(Rotation=(X=0,Y=0,Z=0,W=1),Translation=(X=-2,Y=0,Z=1),Scale3D=(X=1,Y=1,Z= ...(truncated)"
    ItemToModify.1.Transform.Rotation=&RigVMFunction_MathQuaternionFromEuler(Euler=(X=0,Y=10,Z=0)).Result
    ItemToModify.2.Transform.Rotation=&RigVMFunction_MathQuaternionFromEuler(Euler=(X=0,Y=-10,Z=0)).Result
    ItemToModify.0.Transform.Rotation=&RigVMFunction_MathQuaternionFromEuler(Euler=(X=0,Y=0,Z=0)).Result
    Weight=@n17.Result
   )
   > &RigUnit_ModifyTransforms(
    ItemToModify="((Item=(Type=Bone,Name=\"dyn_hair_r_01\"),Transform=(Rotation=(X=0,Y=0,Z=0,W=1),Translation=(X=-2,Y=0,Z=1),Scale3D=(X=1,Y=1,Z=1))),(Item=(Type=Bone,Name=\"dyn_hair_bk_r_01\"),Transform=(Rotation=(X=0,Y=0,Z=0,W=1),Translation=(X=0,Y=0,Z=1),Scale3D=(X=1,Y=1,Z=1) ...(truncated)"
    ItemToModify.1.Transform.Rotation=&RigVMFunction_MathQuaternionFromEuler(Euler=(X=0,Y=0,Z=-30)).Result
    ItemToModify.2.Transform.Rotation=&RigVMFunction_MathQuaternionFromEuler(Euler=(X=0,Y=0,Z=0)).Result
    ItemToModify.0.Transform.Rotation=&RigVMFunction_MathQuaternionFromEuler(Euler=(X=0,Y=0,Z=0)).Result
    Weight=@n17.Result
   )
   > n18 = &RigUnit_SphericalPoseReader(
    ActiveRegionScaleFactors=(PositiveWidth=0.48,NegativeWidth=0.2,PositiveHeight=1,NegativeHeight=1)
    ActiveRegionSize=0.2
    Debug=(bDrawDebug=false,DebugScale=25,DebugSegments=20,DebugThickness=0.25)
    DriverAxis=(X=1,Y=0,Z=0)
    DriverItem=(Type=Bone,Name="head")
    FalloffRegionScaleFactors=(PositiveWidth=0.28,NegativeWidth=0.1,PositiveHeight=0.4,NegativeHeight=0.2)
    FalloffSize=0.5
    OptionalParentItem=(Type=Bone,Name="neck_01")
    RotationOffset=(X=-90,Y=0,Z=55)
   )
   n19 = &RigVMFunction_AlphaInterp(
    InterpSpeedDecreasing=10
    InterpSpeedIncreasing=5
    bClampResult=False
    Value=@n18.OutputParam
    Scale=RigVMVariableNode() {variable=Moving}.Value
   )
   > n20 = &RigUnit_ModifyTransforms(
    ItemToModify="((Item=(Type=Bone,Name=\"dyn_hair_r_01\"),Transform=(Rotation=(X=0,Y=0,Z=0,W=1),Translation=(X=-2,Y=5,Z=0),Scale3D=(X=1,Y=1,Z=1))),(Item=(Type=Bone,Name=\"dyn_hair_bk_r_01\"),Transform=(Rotation=(X=0,Y=0,Z=0,W=1),Translation=(X=0,Y=0,Z=1),Scale3D=(X=1,Y=1,Z=1) ...(truncated)"
    Weight=@n19.Result
    ItemToModify.1.Transform.Rotation=&RigVMFunction_MathQuaternionFromEuler(Euler=(X=0,Y=-20,Z=0)).Result
    ItemToModify.2.Transform.Rotation=&RigVMFunction_MathQuaternionFromEuler(Euler=(X=0,Y=0,Z=0)).Result
    ItemToModify.0.Transform.Rotation=&RigVMFunction_MathQuaternionFromEuler(Euler=(X=0,Y=0,Z=-30)).Result
   )
   Branch_1_1_2_RigVMUnitNode = &RigVMFunction_ControlFlowBranch(ExecuteContext=@n20.ExecutePin, Condition=RigVMVariableNode() {variable=ADS_Rifle}.Value)
   Branch_1_1_1_1_RigVMUnitNode = &RigVMFunction_ControlFlowBranch(ExecuteContext=@Branch_1_1_2_RigVMUnitNode.False, Condition=RigVMVariableNode() {variable=ADS_Pistol}.Value)
   &RigUnit_ParentConstraint(
    AdvancedSettings=(RotationOrderForFilter=XZY)
    Child=(Type=Bone,Name="dyn_hair_bk_02")
    Filter=(TranslationFilter=(bX=True,bY=True,bZ=True),RotationFilter=(bX=True,bY=True,bZ=True),ScaleFilter=(bX=false,bY=false,bZ=false))
    Parents=((Item=(Type=Bone,Name="spine_05"),Weight=1))
    Weight=1
    ExecutePin=@Branch_1_1_1_1_RigVMUnitNode.True
   )
   > &RigUnit_ParentConstraint(
    AdvancedSettings=(RotationOrderForFilter=XZY)
    Child=(Type=Bone,Name="dyn_hair_bk_r_03")
    Filter=(TranslationFilter=(bX=True,bY=True,bZ=True),RotationFilter=(bX=True,bY=True,bZ=True),ScaleFilter=(bX=false,bY=false,bZ=false))
    Parents=((Item=(Type=Bone,Name="spine_05"),Weight=1))
    Weight=1
   )
   > &RigUnit_ParentConstraint(
    AdvancedSettings=(RotationOrderForFilter=XZY)
    Child=(Type=Bone,Name="dyn_hair_bk_l_03")
    Filter=(TranslationFilter=(bX=True,bY=True,bZ=True),RotationFilter=(bX=True,bY=True,bZ=True),ScaleFilter=(bX=false,bY=false,bZ=false))
    Parents=((Item=(Type=Bone,Name="spine_05"),Weight=1))
    Weight=1
   )
   > &RigUnit_ParentConstraint(
    AdvancedSettings=(RotationOrderForFilter=XZY)
    Child=(Type=Bone,Name="dyn_hair_r_03")
    Filter=(TranslationFilter=(bX=true,bY=true,bZ=true),RotationFilter=(bX=True,bY=True,bZ=True),ScaleFilter=(bX=false,bY=false,bZ=false))
    Parents=((Item=(Type=Bone,Name="spine_05"),Weight=1))
    Weight=1
   )
   > n21 = &RigUnit_SphericalPoseReader(
    ActiveRegionScaleFactors=(PositiveWidth=0.45,NegativeWidth=0.2,PositiveHeight=1,NegativeHeight=1)
    ActiveRegionSize=0.2
    Debug=(bDrawDebug=false,DebugScale=25,DebugSegments=20,DebugThickness=0.25)
    DriverAxis=(X=1,Y=0,Z=0)
    DriverItem=(Type=Bone,Name="head")
    FalloffRegionScaleFactors=(PositiveWidth=0.2,NegativeWidth=0.1,PositiveHeight=0.4,NegativeHeight=0.2)
    FalloffSize=0.45
    OptionalParentItem=(Type=Bone,Name="neck_01")
    RotationOffset=(X=-90,Y=0,Z=55)
   )
   > &RigUnit_ModifyTransforms(
    ItemToModify="((Item=(Type=Bone,Name=\"dyn_hair_bk_r_01\"),Transform=(Rotation=(X=0,Y=0,Z=0,W=1),Translation=(X=-5,Y=0,Z=0),Scale3D=(X=1,Y=1,Z=1))),(Item=(Type=Bone,Name=\"dyn_hair_r_01\"),Transform=(Rotation=(X=0,Y=0,Z=0,W=1),Translation=(X=0,Y=0,Z=0),Scale3D=(X=1,Y=1,Z=1) ...(truncated)"
    ItemToModify.0.Transform.Rotation=&RigVMFunction_MathQuaternionFromEuler(Euler=(X=0,Y=-20,Z=0)).Result
    ItemToModify.1.Transform.Rotation=&RigVMFunction_MathQuaternionFromEuler(Euler=(X=0,Y=-5,Z=0)).Result
    ItemToModify.2.Transform.Rotation=&RigVMFunction_MathQuaternionFromEuler(Euler=(X=0,Y=0,Z=0)).Result
    Weight=@n21.OutputParam
   )
   > n22 = &RigUnit_SphericalPoseReader(
    ActiveRegionScaleFactors=(PositiveWidth=0.5,NegativeWidth=0.5,PositiveHeight=0.5,NegativeHeight=0.5)
    ActiveRegionSize=0.15
    Debug=(bDrawDebug=true,DebugScale=25,DebugSegments=20,DebugThickness=0.25)
    DriverAxis=(X=1,Y=0,Z=0)
    DriverItem=(Type=Bone,Name="head")
    FalloffRegionScaleFactors=(PositiveWidth=0.3,NegativeWidth=0.05,PositiveHeight=0.5,NegativeHeight=0.5)
    FalloffSize=0.4
    OptionalParentItem=(Type=Bone,Name="None")
    RotationOffset=(X=100,Y=0,Z=-70)
   )
   > &RigUnit_ModifyTransforms(
    ItemToModify="((Item=(Type=Bone,Name=\"dyn_hair_fr_l_02\"),Transform=(Rotation=(X=0,Y=0,Z=0,W=1),Translation=(X=0,Y=0,Z=0),Scale3D=(X=1,Y=1,Z=1))),(Item=(Type=Bone,Name=\"dyn_hair_bk_01\"),Transform=(Rotation=(X=0,Y=0,Z=0,W=1),Translation=(X=0,Y=0,Z=0),Scale3D=(X=1,Y=1,Z=1) ...(truncated)"
    ItemToModify.4.Transform.Rotation=&RigVMFunction_MathQuaternionFromEuler(Euler=(X=0,Y=0,Z=-30), RotationOrder=XYZ).Result
    ItemToModify.3.Transform.Rotation=&RigVMFunction_MathQuaternionFromEuler(Euler=(X=0,Y=25,Z=0), RotationOrder=XYZ).Result
    ItemToModify.2.Transform.Rotation=&RigVMFunction_MathQuaternionFromEuler(Euler=(X=0,Y=0,Z=0), RotationOrder=XYZ).Result
    ItemToModify.1.Transform.Rotation=&RigVMFunction_MathQuaternionFromEuler(Euler=(X=0,Y=0,Z=0), RotationOrder=XYZ).Result
    ItemToModify.0.Transform.Rotation=&RigVMFunction_MathQuaternionFromEuler(Euler=(X=0,Y=0,Z=0), RotationOrder=XYZ).Result
    Weight=@n22.OutputParam
   )
   n23 = &RigVMFunction_AlphaInterp(InterpSpeedDecreasing=10, InterpSpeedIncreasing=10, Scale=1, bClampResult=False, Value=RigVMVariableNode() {variable=Moving}.Value)
   > &RigUnit_ParentConstraint(
    AdvancedSettings=(RotationOrderForFilter=XZY)
    Child=(Type=Bone,Name="dyn_hair_bk_02")
    Filter=(TranslationFilter=(bX=True,bY=True,bZ=True),RotationFilter=(bX=True,bY=True,bZ=True),ScaleFilter=(bX=false,bY=false,bZ=false))
    Parents=((Item=(Type=Bone,Name="neck_01"),Weight=1))
    Weight=@n23.Result
   )
   > &RigUnit_ParentConstraint(
    AdvancedSettings=(RotationOrderForFilter=XZY)
    Child=(Type=Bone,Name="dyn_hair_bk_03")
    Filter=(TranslationFilter=(bX=True,bY=True,bZ=True),RotationFilter=(bX=True,bY=True,bZ=True),ScaleFilter=(bX=false,bY=false,bZ=false))
    Parents=((Item=(Type=Bone,Name="spine_05"),Weight=1))
    Weight=@n23.Result
   )
   > &RigUnit_ParentConstraint(
    AdvancedSettings=(RotationOrderForFilter=XZY)
    Child=(Type=Bone,Name="dyn_hair_r_01")
    Filter=(TranslationFilter=(bX=false,bY=false,bZ=false),RotationFilter=(bX=True,bY=True,bZ=True),ScaleFilter=(bX=false,bY=false,bZ=false))
    Parents=((Item=(Type=Bone,Name="spine_05"),Weight=1))
    Weight=@n23.Result
   )
   > &RigUnit_ParentConstraint(
    AdvancedSettings=(RotationOrderForFilter=XZY)
    Child=(Type=Bone,Name="dyn_hair_bk_l_03")
    Filter=(TranslationFilter=(bX=True,bY=True,bZ=True),RotationFilter=(bX=True,bY=True,bZ=True),ScaleFilter=(bX=false,bY=false,bZ=false))
    Parents=((Item=(Type=Bone,Name="spine_05"),Weight=1))
    Weight=@n23.Result
   )
   > &RigUnit_ParentConstraint(
    AdvancedSettings=(RotationOrderForFilter=XZY)
    Child=(Type=Bone,Name="dyn_hair_l_03")
    Filter=(TranslationFilter=(bX=True,bY=True,bZ=True),RotationFilter=(bX=True,bY=True,bZ=True),ScaleFilter=(bX=false,bY=false,bZ=false))
    Parents=((Item=(Type=Bone,Name="neck_01"),Weight=1))
    Weight=@n23.Result
   )
   &RigUnit_ParentConstraint(
    AdvancedSettings=(RotationOrderForFilter=XZY)
    Child=(Type=Bone,Name="dyn_hair_bk_02")
    Filter=(TranslationFilter=(bX=True,bY=True,bZ=True),RotationFilter=(bX=True,bY=True,bZ=True),ScaleFilter=(bX=false,bY=false,bZ=false))
    Parents=((Item=(Type=Bone,Name="spine_05"),Weight=1))
    Weight=1
    ExecutePin=@Branch_1_1_2_RigVMUnitNode.True
   )
   > &RigUnit_ParentConstraint(
    AdvancedSettings=(RotationOrderForFilter=XZY)
    Child=(Type=Bone,Name="dyn_hair_bk_r_03")
    Filter=(TranslationFilter=(bX=True,bY=True,bZ=True),RotationFilter=(bX=True,bY=True,bZ=True),ScaleFilter=(bX=false,bY=false,bZ=false))
    Parents=((Item=(Type=Bone,Name="spine_05"),Weight=1))
    Weight=1
   )
   > &RigUnit_ParentConstraint(
    AdvancedSettings=(RotationOrderForFilter=XZY)
    Child=(Type=Bone,Name="dyn_hair_bk_l_03")
    Filter=(TranslationFilter=(bX=True,bY=True,bZ=True),RotationFilter=(bX=True,bY=True,bZ=True),ScaleFilter=(bX=false,bY=false,bZ=false))
    Parents=((Item=(Type=Bone,Name="spine_05"),Weight=1))
    Weight=1
   )
   > &RigUnit_ParentConstraint(
    AdvancedSettings=(RotationOrderForFilter=XZY)
    Child=(Type=Bone,Name="dyn_hair_r_03")
    Filter=(TranslationFilter=(bX=true,bY=true,bZ=true),RotationFilter=(bX=True,bY=True,bZ=True),ScaleFilter=(bX=false,bY=false,bZ=false))
    Parents=((Item=(Type=Bone,Name="spine_05"),Weight=1))
    Weight=1
   )
   > n24 = &RigUnit_SphericalPoseReader(
    ActiveRegionScaleFactors=(PositiveWidth=0.1,NegativeWidth=0.7,PositiveHeight=1,NegativeHeight=1)
    ActiveRegionSize=0.2
    Debug=(bDrawDebug=false,DebugScale=25,DebugSegments=20,DebugThickness=0.25)
    DriverAxis=(X=1,Y=0,Z=0)
    DriverItem=(Type=Bone,Name="head")
    FalloffRegionScaleFactors=(PositiveWidth=0.1,NegativeWidth=0.2,PositiveHeight=0.5,NegativeHeight=0.5)
    FalloffSize=0.45
    OptionalParentItem=(Type=Bone,Name="neck_01")
    RotationOffset=(X=-90,Y=0,Z=120)
   )
   > &RigUnit_ParentConstraint(
    AdvancedSettings=(RotationOrderForFilter=XZY)
    Child=(Type=Bone,Name="dyn_hair_fr_l_05")
    Filter=(TranslationFilter=(bX=true,bY=true,bZ=false),RotationFilter=(bX=True,bY=True,bZ=True),ScaleFilter=(bX=false,bY=false,bZ=false))
    Parents=((Item=(Type=Bone,Name="neck_01"),Weight=0.05))
    Weight=0.5
   )
   > &RigUnit_ModifyTransforms(
    ItemToModify="((Item=(Type=Bone,Name=\"dyn_hair_fr_l_02\"),Transform=(Rotation=(X=0,Y=0,Z=0,W=1),Translation=(X=0,Y=0,Z=0),Scale3D=(X=1,Y=1,Z=1))),(Item=(Type=Bone,Name=\"dyn_hair_fr_l_06\"),Transform=(Rotation=(X=0,Y=0,Z=0,W=1),Translation=(X=0,Y=0,Z=0),Scale3D=(X=1,Y=1,Z= ...(truncated)"
    ItemToModify.4.Transform.Rotation=&RigVMFunction_MathQuaternionFromEuler(Euler=(X=0,Y=0,Z=0), RotationOrder=XYZ).Result
    ItemToModify.3.Transform.Rotation=&RigVMFunction_MathQuaternionFromEuler(Euler=(X=0,Y=0,Z=10), RotationOrder=XYZ).Result
    ItemToModify.2.Transform.Rotation=&RigVMFunction_MathQuaternionFromEuler(Euler=(X=0,Y=20,Z=0), RotationOrder=XYZ).Result
    Weight=@n24.OutputParam
   )
   > n25 = &RigUnit_SphericalPoseReader(
    ActiveRegionScaleFactors=(PositiveWidth=0.48,NegativeWidth=0.2,PositiveHeight=1,NegativeHeight=1)
    ActiveRegionSize=0.2
    Debug=(bDrawDebug=true,DebugScale=25,DebugSegments=20,DebugThickness=0.25)
    DriverAxis=(X=1,Y=0,Z=0)
    DriverItem=(Type=Bone,Name="head")
    FalloffRegionScaleFactors=(PositiveWidth=0.25,NegativeWidth=0.1,PositiveHeight=0.4,NegativeHeight=0.2)
    FalloffSize=0.5
    OptionalParentItem=(Type=Bone,Name="neck_01")
    RotationOffset=(X=-90,Y=0,Z=55)
   )
   > &RigUnit_ModifyTransforms(
    ItemToModify="((Item=(Type=Bone,Name=\"dyn_hair_fr_l_02\"),Transform=(Rotation=(X=0,Y=0,Z=0,W=1),Translation=(X=0,Y=0,Z=0),Scale3D=(X=1,Y=1,Z=1))),(Item=(Type=Bone,Name=\"dyn_hair_fr_r_01\"),Transform=(Rotation=(X=0,Y=0,Z=0,W=1),Translation=(X=-2,Y=0,Z=0),Scale3D=(X=1,Y=1,Z ...(truncated)"
    ItemToModify.4.Transform.Rotation=&RigVMFunction_MathQuaternionFromEuler(Euler=(X=0,Y=0,Z=0), RotationOrder=XYZ).Result
    ItemToModify.3.Transform.Rotation=&RigVMFunction_MathQuaternionFromEuler(Euler=(X=0,Y=0,Z=10), RotationOrder=XYZ).Result
    ItemToModify.2.Transform.Rotation=&RigVMFunction_MathQuaternionFromEuler(Euler=(X=0,Y=20,Z=0), RotationOrder=XYZ).Result
    ItemToModify.1.Transform.Rotation=&RigVMFunction_MathQuaternionFromEuler(Euler=(X=0,Y=0,Z=0), RotationOrder=XYZ).Result
    Weight=@n25.OutputParam
   )
   &RigUnit_ModifyTransforms(
    ItemToModify=((Item=(Type=Bone,Name="dyn_hair_fr_l_01"),Transform=(Rotation=(X=0,Y=0,Z=0,W=1),Translation=(X=0,Y=0,Z=0),Scale3D=(X=1,Y=1,Z=1))))
    Weight=1
    ExecutePin=@Branch_1_RigVMUnitNode.True
    ItemToModify.0.Transform.Rotation=&RigVMFunction_MathQuaternionFromEuler(Euler=(X=0,Y=-10,Z=0)).Result
   )
   > &RigUnit_ModifyTransforms(
    ItemToModify="((Item=(Type=Bone,Name=\"dyn_hair_bk_01\"),Transform=(Rotation=(X=0,Y=0,Z=0,W=1),Translation=(X=0,Y=0,Z=8),Scale3D=(X=1,Y=1,Z=1))),(Item=(Type=Bone,Name=\"dyn_hair_bk_l_01\"),Transform=(Rotation=(X=0,Y=0,Z=0,W=1),Translation=(X=0,Y=0,Z=8),Scale3D=(X=1,Y=1,Z=1) ...(truncated)"
    Weight=1
    ItemToModify.1.Transform.Rotation=&RigVMFunction_MathQuaternionFromEuler(Euler=(X=0,Y=-10,Z=0)).Result
    ItemToModify.2.Transform.Rotation=&RigVMFunction_MathQuaternionFromEuler(Euler=(X=0,Y=-10,Z=0)).Result
    ItemToModify.0.Transform.Rotation=&RigVMFunction_MathQuaternionFromEuler(Euler=(X=0,Y=-10,Z=0)).Result
   )
   > &RigUnit_ModifyTransforms(
    ItemToModify=((Item=(Type=Bone,Name="dyn_hair_r_01"),Transform=(Rotation=(X=0,Y=0,Z=0,W=1),Translation=(X=0,Y=2,Z=5),Scale3D=(X=1,Y=1,Z=1))),(Item=(Type=Bone,Name="dyn_hair_l_01"),Transform=(Rotation=(X=0,Y=0,Z=0,W=1),Translation=(X=0,Y=0,Z=0),Scale3D=(X=1,Y=1,Z=1))))
    Weight=1
    ItemToModify.0.Transform.Rotation=&RigVMFunction_MathQuaternionFromEuler(Euler=(X=0,Y=0,Z=-30)).Result
    ItemToModify.1.Transform.Rotation=&RigVMFunction_MathQuaternionFromEuler(Euler=(X=0,Y=0,Z=-30)).Result
   )
   &RigVMFunction_MathQuaternionFromEuler(Euler=(X=0,Y=0,Z=0), RotationOrder=XYZ)
   &RigVMFunction_MathQuaternionFromEuler(Euler=(X=0,Y=0,Z=0), RotationOrder=XYZ)
  ]}
]
hierarchy [
 root {loc=<0 0 0>, rot=<0 0 0 1>}
 pelvis {parent=root, loc=<0 -0.5301 98.0571>, rot=<0.0127 0.707 -0.0127 -0.707>}
 spine_01 {parent=pelvis, loc=<6.1208 -0.9328 0>, rot=<0 0 0.034 -0.9994>}
 spine_02 {parent=spine_01, loc=<6.5996 -0.2095 0>, rot=<0 0 0.0238 -0.9997>}
 spine_03 {parent=spine_02, loc=<6.4914 -0.1237 0>, rot=<0 0 0.0149 -0.9999>}
 spine_04 {parent=spine_03, loc=<6.9598 0.0794 0>, rot=<0 0 -0.0132 -0.9999>}
 spine_05 {parent=spine_04, loc=<6.9987 0 0>, rot=<0 0 -0.0015 -1>}
 neck_01 {parent=spine_05, loc=<11.3138 0.1529 0>, rot=<0 0 -0.2003 -0.9797>}
 neck_02 {parent=neck_01, loc=<4.5627 0 0>, rot=<0 0 0.0504 -0.9987>}
 head {parent=neck_02, loc=<5.3914 0 0>, rot=<0 0 0.115 -0.9934>}
 dyn_hair_fr_l_01 {parent=head, loc=<12.119 8.6928 -4.2882>, rot=<0.1826 -0.9444 0.0497 -0.2687>}
 dyn_hair_fr_l_02 {parent=dyn_hair_fr_l_01, loc=<4.1367 0 0>, rot=<0.0162 -0.0389 0.0207 -0.9989>}
 dyn_hair_fr_l_03 {parent=dyn_hair_fr_l_02, loc=<6.9128 0 0>, rot=<0.0085 -0.0188 0.012 -0.9997>}
 dyn_hair_fr_l_04 {parent=dyn_hair_fr_l_03, loc=<5.9829 0 0>, rot=<0.1063 -0.4667 -0.1876 -0.8577>}
 dyn_hair_fr_l_05 {parent=dyn_hair_fr_l_04, loc=<4.6765 0 0>, rot=<0.0179 0.1151 -0.1609 -0.9801>}
 dyn_hair_fr_l_06 {parent=dyn_hair_fr_l_05, loc=<3.7727 0 0>, rot=<-0.0268 -0.0752 -0.0252 -0.9965>}
 dyn_hair_fr_l_07 {parent=dyn_hair_fr_l_06, loc=<3.7274 0 0>, rot=<0.008 0.0309 0.113 -0.9931>}
 dyn_hair_l_01 {parent=head, loc=<5.8744 -2.5447 -8.9664>, rot=<0.177 -0.9801 0.0153 -0.0887>}
 dyn_hair_l_02 {parent=dyn_hair_l_01, loc=<4.6736 0 0>, rot=<0.0398 -0.1768 -0.1252 -0.9754>}
 dyn_hair_l_03 {parent=dyn_hair_l_02, loc=<4.9167 0 0>, rot=<-0.0078 0.0533 0.0523 -0.9972>}
 dyn_hair_l_04 {parent=dyn_hair_l_03, loc=<4.3471 0 0>, rot=<-0.1192 0.3228 0.1453 -0.9276>}
 dyn_hair_bk_l_01 {parent=head, loc=<2.0125 -8.0093 -6.7865>, rot=<0.0578 -0.9125 0.405 0.0007>}
 dyn_hair_bk_l_02 {parent=dyn_hair_bk_l_01, loc=<6.1978 0.1674 -0.6726>, rot=<-0.0227 0.1374 0.0279 -0.9899>}
 dyn_hair_bk_l_03 {parent=dyn_hair_bk_l_02, loc=<6.0851 -0.5195 0.2676>, rot=<0.0104 -0.1014 0.0438 -0.9938>}
 dyn_hair_bk_l_04 {parent=dyn_hair_bk_l_03, loc=<4.5006 -0.0302 0>, rot=<0 0 -0.0286 -0.9996>}
 dyn_hair_bk_01 {parent=head, loc=<0.0961 -8.8034 -1.3714>, rot=<0.0667 -0.7372 0.6499 -0.1726>}
 dyn_hair_bk_02 {parent=dyn_hair_bk_01, loc=<6.6433 -0.1392 0.1687>, rot=<0.0208 -0.0259 0.0629 -0.9975>}
 dyn_hair_bk_03 {parent=dyn_hair_bk_02, loc=<8.9698 -0.5773 -0.0494>, rot=<-0.0158 -0.2463 -0.0479 -0.9679>}
 dyn_hair_bk_r_01 {parent=head, loc=<1.587 -6.8709 4.8773>, rot=<0.0903 -0.6498 0.7422 -0.1373>}
 dyn_hair_bk_r_02 {parent=dyn_hair_bk_r_01, loc=<6.2815 0.3916 -0.2562>, rot=<0.002 0.0492 -0.0384 -0.998>}
 dyn_hair_bk_r_03 {parent=dyn_hair_bk_r_02, loc=<5.703 0.1962 0.2339>, rot=<-0.0664 -0.0545 -0.068 -0.994>}
 dyn_hair_bk_r_04 {parent=dyn_hair_bk_r_03, loc=<4.1499 -0.1579 -0.0227>, rot=<0.0172 -0.1959 -0.0025 -0.9805>}
 dyn_hair_r_01 {parent=head, loc=<2.6559 -6.601 6.151>, rot=<0.1558 -0.3943 0.9009 -0.0933>}
 dyn_hair_r_02 {parent=dyn_hair_r_01, loc=<6.2128 -0.2206 0.1106>, rot=<-0.024 -0.0368 -0.0691 -0.9966>}
 dyn_hair_r_03 {parent=dyn_hair_r_02, loc=<5.9815 0 0>, rot=<-0.0399 -0.1573 -0.0572 -0.9851>}
 dyn_hair_r_04 {parent=dyn_hair_r_03, loc=<4.7285 0 0>, rot=<-0.0366 -0.1135 -0.1391 -0.9831>}
 dyn_hair_fr_r_01 {parent=head, loc=<6.6006 -2.1619 7.2146>, rot=<-0.0647 0.0083 0.9801 -0.1875>}
 dyn_hair_fr_r_02 {parent=dyn_hair_fr_r_01, loc=<3.0709 0 0>, rot=<0.0637 0.2491 0.1172 -0.9592>}
 dyn_hair_fr_r_03 {parent=dyn_hair_fr_r_02, loc=<3.6187 0 0>, rot=<0.0032 0.0764 0.088 -0.9932>}
 dyn_hair_fr_r_04 {parent=dyn_hair_fr_r_03, loc=<3.3408 0 0>, rot=<0.015 -0.1874 0.032 -0.9817>}
 dyn_hair_fr_r_05 {parent=dyn_hair_fr_r_04, loc=<2.6642 0 0>, rot=<0.021 -0.2068 -0.0137 -0.9781>}
 faceAttach {parent=head, loc=<-1.3858 -0.0582 0>, rot=<0.0029 -0.7071 -0.0029 -0.7071>}
 C_brow_mid {parent=faceAttach, loc=<-0.0028 9.9447 5.4778>, rot=<0 0 0 -1>}
 L_brow_mid {parent=faceAttach, loc=<3.1749 9.5186 6.0223>, rot=<-0.0001 -0.0042 0.2553 -0.9669>}
 L_brow_outer {parent=faceAttach, loc=<5.921 6.622 5.7915>, rot=<-0.0287 -0.04 0.4491 -0.8921>}
 R_brow_mid {parent=faceAttach, loc=<-3.1863 9.523 6.0158>, rot=<-0.0001 0.0042 -0.2553 -0.9669>}
 R_brow_outer {parent=faceAttach, loc=<-5.9116 6.64 5.7337>, rot=<-0.0287 0.04 -0.4491 -0.8921>}
 dyn_hair_midr_01 {parent=head, loc=<2.577 -5.427 7.4435>, rot=<0.1198 -0.0516 0.9902 -0.0497>}
 dyn_hair_midr_02 {parent=dyn_hair_midr_01, loc=<2.7381 -0.2622 -0.1009>, rot=<-0.024 -0.0368 -0.0691 -0.9966>}
 dyn_hair_midr_03 {parent=dyn_hair_midr_02, loc=<3.7583 -0.5415 -0.1604>, rot=<-0.0422 -0.0355 0.0962 -0.9938>}
 dyn_hair_midr_04 {parent=dyn_hair_midr_03, loc=<2.6907 -0.0341 0.169>, rot=<0.0561 0.1205 -0.0053 -0.9911>}
 clavicle_l {parent=spine_05, loc=<5.4831 -0.1053 -3.2946>, rot=<0.0465 -0.7289 -0.009 -0.6829>}
 upperarm_l {parent=clavicle_l, loc=<9.8658 -1.2136 -0.4105>, rot=<0.0164 -0.3245 -0.0269 -0.9454>}
 clavicle_r {parent=spine_05, loc=<5.4831 -0.1053 3.2946>, rot=<0.7289 0.0465 -0.6829 0.009>}
 upperarm_r {parent=clavicle_r, loc=<-9.8658 1.2136 0.4105>, rot=<0.0164 -0.3245 -0.0269 -0.9454>}
]
variables [
 ADS_Pistol {type=bool, default=False}
 ADS_Rifle {type=bool, default=False}
 Blazer {type=double, default=0}
 Crouch {type=bool, default=False}
 Emote {type=double, default=0}
 Melee {type=double, default=0}
 Moving {type=double, default=0}
 Rifle {type=double, default=0}
 Sailor {type=double, default=0}
 Skydive {type=bool, default=False}
 Sprint {type=double, default=0}
]
