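# Physics asset for the hair of the Pencil Pear medium female player head, set up for hair dynamics against the upper body.
# Four capsule bone chains (back, back-left, back-right and right side) hang from the head, each link constrained to its parent with moderate swing and very tight twist limits.
# A head sphere and capsules on the neck, upper spine and clavicles supply body shapes, joined by free-motion constraints except spine_04, which has none.
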PhysicsAsset /BRCosmetics/Characters/Player/Female/Medium/Heads/F_MED_Pencil_Pear_Head/Meshes/Hair/F_MED_Pencil_Pear_Hair_PhysicsAsset
bodies {
 clavicle_l [{
   shapeType=Capsule
   length=27.2273
   radius=7.192
   shapeName=clavicle_l_capsule
   center=<10.068 -2.1924 -2.1893>
   rotation=<rot -14.9347 92.402 -91.3372>
  }]
 clavicle_r [{
   shapeType=Capsule
   length=27.2273
   radius=7.192
   shapeName=clavicle_r_capsule
   center=<-9.4123 1.9516 1.1842>
   rotation=<rot -14.9347 92.402 -91.3372>
  }]
 dyn_hair_bk_01 [{shapeType=Capsule, length=3, radius=2, shapeName=dyn_hair_bk_01_capsule, rotation=<rot 87.1875 -89.9998 -89.9998>}]
 dyn_hair_bk_02 [{shapeType=Capsule, length=5.3463, radius=2.5542, shapeName=dyn_hair_bk_02_capsule, center=<3 0 0>, rotation=<rot 90 0 0>}]
 dyn_hair_bk_03 [{
   shapeType=Capsule
   length=5.1509
   radius=2.8786
   shapeName=dyn_hair_bk_03_capsule
   center=<5.2563 0 0.6046>
   rotation=<rot 75.0001 -179.9999 -180>
  }]
 dyn_hair_bk_l_01 [{shapeType=Capsule, length=3, radius=2, shapeName=dyn_hair_bk_l_01_capsule, rotation=<rot 87.1875 -89.9998 -89.9998>}]
 dyn_hair_bk_l_02 [{shapeType=Capsule, length=7, radius=2, shapeName=dyn_hair_bk_l_02_capsule, center=<3 0 0>, rotation=<rot 90 0 0>}]
 dyn_hair_bk_l_03 [{shapeType=Capsule, length=7, radius=2, shapeName=dyn_hair_bk_l_03_capsule, center=<3 0 0>, rotation=<rot 90 0 0>}]
 dyn_hair_bk_l_04 [{shapeType=Capsule, length=7, radius=2, shapeName=dyn_hair_bk_l_04_capsule, center=<3 0 0>, rotation=<rot 90 0 0>}]
 dyn_hair_bk_r_01 [{shapeType=Capsule, length=4, radius=2, shapeName=dyn_hair_bk_r_01_capsule, center=<3 0 0>, rotation=<rot 87.1875 -89.9998 -89.9998>}]
 dyn_hair_bk_r_02 [{shapeType=Capsule, length=7, radius=2, shapeName=dyn_hair_bk_r_02_capsule, center=<2 0 0>, rotation=<rot 90 0 0>}]
 dyn_hair_bk_r_03 [{shapeType=Capsule, length=7, radius=2, shapeName=dyn_hair_bk_r_03_capsule, center=<2 0 0>, rotation=<rot 90 0 0>}]
 dyn_hair_bk_r_04 [{shapeType=Capsule, length=7, radius=2, shapeName=dyn_hair_bk_r_04_capsule, center=<2 0 0>, rotation=<rot 90 0 0>}]
 dyn_hair_r_01 [{shapeType=Capsule, length=4, radius=2, shapeName=dyn_hair_r_01_capsule, center=<3 0 0>, rotation=<rot 90 0 0>}]
 dyn_hair_r_02 [{shapeType=Capsule, length=4.411, radius=2, shapeName=dyn_hair_r_02_capsule, center=<2 0 0>, rotation=<rot 0 0 90>}]
 dyn_hair_r_03 [{shapeType=Capsule, length=8, radius=3, shapeName=dyn_hair_r_03_capsule, center=<3 0 0>, rotation=<rot 90 0 0>}]
 dyn_hair_r_04 [{shapeType=Capsule, length=5.0347, radius=2.5, shapeName=dyn_hair_r_04_capsule, center=<1 0 0>, rotation=<rot 90 -49.0963 -44.0963>}]
 head [{shapeType=Sphere, radius=8.7086, shapeName=head_sphere, center=<2.2822 2.2545 -0.0836>}]
 neck_01 [{shapeType=Capsule, length=10.5078, radius=8.8517, shapeName=neck_01_capsule, center=<-2.531 1.7189 0>, rotation=<rot 90 0 16.8752>}]
 neck_02 [{shapeType=Capsule, length=12.8371, radius=6.8539, shapeName=neck_02_capsule, rotation=<rot 90 32.0237 32.0242>}]
 spine_04 [{shapeType=Capsule, length=16.7896, radius=13.2104, center=<-0.5935 3.9724 -0.6655>, rotation=<rot 89.3496 -174.3751 180>}]
 spine_05 [{shapeType=Capsule, length=31.5915, radius=11.6443, center=<-0.0301 2.9073 0.0001>}]
}
constraints [
 {
  bone1Name=head
  bone2Name=neck_02
  swing1LimitDegrees=45
  swing1Motion=Free
  swing2LimitDegrees=45
  swing2Motion=Free
  twistLimitDegrees=45
  twistMotion=Free
 }
 {
  bone1Name=neck_02
  bone2Name=neck_01
  swing1LimitDegrees=45
  swing1Motion=Free
  swing2LimitDegrees=45
  swing2Motion=Free
  twistLimitDegrees=45
  twistMotion=Free
 }
 {
  bone1Name=neck_01
  bone2Name=spine_05
  swing1LimitDegrees=45
  swing1Motion=Free
  swing2LimitDegrees=45
  swing2Motion=Free
  twistLimitDegrees=45
  twistMotion=Free
 }
 {bone1Name=dyn_hair_bk_01, bone2Name=head, swing1LimitDegrees=20, swing2LimitDegrees=45, twistLimitDegrees=1}
 {bone1Name=dyn_hair_bk_02, bone2Name=dyn_hair_bk_01, swing1LimitDegrees=30, swing2LimitDegrees=45, twistLimitDegrees=1}
 {bone1Name=dyn_hair_bk_03, bone2Name=dyn_hair_bk_02, swing1LimitDegrees=30, swing2LimitDegrees=45, twistLimitDegrees=1}
 {bone1Name=dyn_hair_bk_l_01, bone2Name=head, swing1LimitDegrees=20, swing2LimitDegrees=45, twistLimitDegrees=1}
 {bone1Name=dyn_hair_bk_l_02, bone2Name=dyn_hair_bk_l_01, swing1LimitDegrees=20, swing2LimitDegrees=45, twistLimitDegrees=1}
 {bone1Name=dyn_hair_bk_l_03, bone2Name=dyn_hair_bk_l_02, swing1LimitDegrees=20, swing2LimitDegrees=45, twistLimitDegrees=1}
 {bone1Name=dyn_hair_bk_l_04, bone2Name=dyn_hair_bk_l_03, swing1LimitDegrees=20, swing2LimitDegrees=45, twistLimitDegrees=1}
 {bone1Name=dyn_hair_bk_r_01, bone2Name=head, swing1LimitDegrees=10, swing2LimitDegrees=30, twistLimitDegrees=2}
 {bone1Name=dyn_hair_bk_r_02, bone2Name=dyn_hair_bk_r_01, swing1LimitDegrees=30, swing2LimitDegrees=45, twistLimitDegrees=1}
 {bone1Name=dyn_hair_bk_r_03, bone2Name=dyn_hair_bk_r_02, swing1LimitDegrees=30, swing2LimitDegrees=45, twistLimitDegrees=1}
 {bone1Name=dyn_hair_bk_r_04, bone2Name=dyn_hair_bk_r_03, swing1LimitDegrees=30, swing2LimitDegrees=45, twistLimitDegrees=1}
 {bone1Name=dyn_hair_r_01, bone2Name=head, swing1LimitDegrees=15, swing2LimitDegrees=30, twistLimitDegrees=1}
 {bone1Name=dyn_hair_r_03, bone2Name=dyn_hair_r_02, swing1LimitDegrees=30, swing2LimitDegrees=30, twistLimitDegrees=1}
 {bone1Name=dyn_hair_r_04, bone2Name=dyn_hair_r_03, swing1LimitDegrees=30, swing2LimitDegrees=30, twistLimitDegrees=1}
 {bone1Name=dyn_hair_r_02, bone2Name=dyn_hair_r_01, swing1LimitDegrees=15, swing2LimitDegrees=30, twistLimitDegrees=1}
 {
  bone1Name=clavicle_r
  bone2Name=spine_05
  swing1LimitDegrees=45
  swing1Motion=Free
  swing2LimitDegrees=45
  swing2Motion=Free
  twistLimitDegrees=45
  twistMotion=Free
 }
 {
  bone1Name=clavicle_l
  bone2Name=spine_05
  swing1LimitDegrees=45
  swing1Motion=Free
  swing2LimitDegrees=45
  swing2Motion=Free
  twistLimitDegrees=45
  twistMotion=Free
 }
]
