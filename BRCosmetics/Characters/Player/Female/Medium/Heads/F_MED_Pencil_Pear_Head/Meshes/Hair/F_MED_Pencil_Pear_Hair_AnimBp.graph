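# Animation Blueprint for the Pencil Pear female head's hair, adding secondary motion to its dynamic bones.
# Per-bone Anim Dynamics with spherical limits and a Look At feed a Control Rig, whose pose is cached.
# Two RigidBody chains simulate that pose, selected by Sailor jacket tag, before constraints and a final Look At.
# The update event sets the rig's movement, emote, weapon, aiming and outfit inputs.
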AnimBlueprint /BRCosmetics/Characters/Player/Female/Medium/Heads/F_MED_Pencil_Pear_Head/Meshes/Hair/F_MED_Pencil_Pear_Hair_AnimBp
target_skeleton=/BRCosmetics/Characters/Player/Common/Fortnite_Base_Head/Pencil/Fortnite_Base_Head_Pencil_Pear_Skeleton
graphs [
 AnimGraph {nodes [
   n1 = RigidBody!(ComponentPose=Local To Component(LocalPose=Use cached pose 'Pose'()), Alpha, OverrideWorldGravity:Vector="0,0,-980", ExternalForce) -> (Pose) {
    bClampLinearTranslationLimitToRefPose=true
    bOverrideWorldGravity=true
    overridePhysicsAsset=&/F_MED_Pencil_Pear_Hair_PhysicsAsset
    simulationSpace=BaseBoneSpace
    alphaBoolBlend {blendOption=Linear}
    baseBoneRef {boneName=spine_05}
    componentLinearAccScale=<0.3 0.3 0.3>
    componentLinearVelScale=<0.3 0.3 0.3>
    overrideWorldGravity {z=-980}
   }
   n2 = RigidBody!(ComponentPose=@n1, bAlphaBoolEnabled:bool, OverrideWorldGravity="0,0,-980", ExternalForce) -> (Pose) {
    alphaInputType=Bool
    bClampLinearTranslationLimitToRefPose=true
    bOverrideWorldGravity=true
    overridePhysicsAsset=&F_MED_Pencil_Pear_Hair_PhysicsAsset
    simulationSpace=BaseBoneSpace
    alphaBoolBlend {blendInTime=0.1, blendOption=Linear, blendOutTime=0.1}
    baseBoneRef {boneName=pelvis}
    componentAppliedLinearAccClamp=<3000 3000 3000>
    componentLinearAccScale=<0.3 0.3 0.4>
    componentLinearVelScale=<0.3 0.3 0.4>
    simSpaceSettings {
     maxAngularAcceleration=1000
     maxAngularVelocity=3
     maxLinearAcceleration=700
     maxLinearVelocity=500
     velocityScaleZ=0
     externalLinearDragV=<0.2 0.2>
    }
   }
   n3 = Component To Local(ComponentPose=@n2)
   n4 = RigidBody!(ComponentPose=Local To Component(LocalPose=Use cached pose 'Pose'()), Alpha:float, OverrideWorldGravity="0,0,-980", ExternalForce) -> (Pose) {
    bClampLinearTranslationLimitToRefPose=true
    bOverrideWorldGravity=true
    overridePhysicsAsset=&/F_MED_Pencil_Pear_Hair_PhysicsAsset_Sailor
    simulationSpace=BaseBoneSpace
    alphaBoolBlend {blendOption=Linear}
    baseBoneRef {boneName=spine_05}
    componentLinearAccScale=<0.3 0.3 0.3>
    componentLinearVelScale=<0.3 0.3 0.3>
    overrideWorldGravity {z=-980}
   }
   n5 = RigidBody!(ComponentPose=@n4, bAlphaBoolEnabled:bool, OverrideWorldGravity="0,0,-980", ExternalForce) -> (Pose) {
    alphaInputType=Bool
    bClampLinearTranslationLimitToRefPose=true
    bOverrideWorldGravity=true
    overridePhysicsAsset=&F_MED_Pencil_Pear_Hair_PhysicsAsset
    simulationSpace=BaseBoneSpace
    alphaBoolBlend {blendInTime=0.1, blendOption=Linear, blendOutTime=0.1}
    baseBoneRef {boneName=pelvis}
    componentAppliedLinearAccClamp=<3000 3000 3000>
    componentLinearAccScale=<0.3 0.3 0.4>
    componentLinearVelScale=<0.3 0.3 0.4>
    simSpaceSettings {
     maxAngularAcceleration=1000
     maxAngularVelocity=3
     maxLinearAcceleration=700
     maxLinearVelocity=500
     velocityScaleZ=0
     externalLinearDragV=<0.2 0.2>
    }
   }
   n6 = Component To Local(ComponentPose=@n5)
   n7 = "Get Sailor_Jacket "() -> ("Sailor_Jacket ":bool)
   n8 = "Constraint - dyn_hair_bk_l_02 (PP)"!(
    ComponentPose=Local To Component(LocalPose=Blend Poses by bool(BlendPose_0=@n6, BlendPose_1=@n3, bActiveValue=@n7) {blendType=Linear})
    Alpha=1
    ConstraintWeights_0=0.8
    ConstraintWeights_1:float=0.2
   ) -> (Pose) {
    alphaBoolBlend {blendOption=Linear}
    boneToModify {boneName=dyn_hair_bk_l_02}
    constraintSetup [{targetBone {boneName=spine_05}}, {targetBone {boneName=neck_01}}]
    constraintWeights [1,1]
   }
   n9 = "Constraint - dyn_hair_bk_02 (PP)"!(ComponentPose=@n8, Alpha=1, ConstraintWeights_0=0.8, ConstraintWeights_1:float=0.2) -> (Pose) {
    alphaBoolBlend {blendOption=Linear}
    boneToModify {boneName=dyn_hair_bk_02}
    constraintSetup [{targetBone {boneName=spine_05}}, {targetBone {boneName=neck_01}}]
    constraintWeights [1,1]
   }
   n10 = "Constraint - dyn_hair_bk_r_02 (PP)"!(ComponentPose=@n9, Alpha=1, ConstraintWeights_0=0.8, ConstraintWeights_1:float=0.2) -> (Pose) {
    alphaBoolBlend {blendOption=Linear}
    boneToModify {boneName=dyn_hair_bk_r_02}
    constraintSetup [{targetBone {boneName=spine_05}}, {targetBone {boneName=neck_01}}]
    constraintWeights [1,1]
   }
   n11 = "Constraint - dyn_hair_bk_r_03 (PP)"!(ComponentPose=@n10, Alpha=1, ConstraintWeights_0=0.8, ConstraintWeights_1:float=0.2) -> (Pose) {
    alphaBoolBlend {blendOption=Linear}
    boneToModify {boneName=dyn_hair_bk_r_03}
    constraintSetup [{targetBone {boneName=spine_05}}, {targetBone {boneName=neck_01}}]
    constraintWeights [1,1]
   }
   n12 = Look At - Bone: dyn_hair_fr_l_02(ComponentPose=@n11, Alpha=1) {
    alphaBoolBlend {blendOption=Linear}
    boneToModify {boneName=dyn_hair_fr_l_02}
    lookAtLocation=<-57.0187 -106.5446 -91.6029>
    lookAtTarget {boneReference {boneName=head}}
    lookAt_Axis {axis=<1 0>}
   }
   Output Pose(Result=Component To Local(ComponentPose=@n12)) {layerGroup=None}
   n13 = Anim Dynamics - Bone: dyn_hair_midr_02!(ComponentPose=Local To Component(LocalPose=Input Pose()), Alpha=0.5, GravityOverride:Vector="0,0,-980", ExternalForce="0,0,0") -> (Pose) {
    angularDampingOverride=0.9999
    angularSpringConstant=75
    bAngularSpring=true
    bOverrideAngularDamping=true
    bUseGravityOverride=true
    bUseSphericalLimits=true
    simulationSpace=RootRelative
    alphaBoolBlend {blendOption=Linear}
    boundBone {boneName=dyn_hair_midr_02}
    componentAppliedLinearAccClamp=<3500 3500 3500>
    componentLinearAccScale=<1 1 1>
    componentLinearVelScale=<1 1 1>
    gravityOverride {z=-980}
    physicsBodyDefinitions [{
      collisionType=CustomSphere
      sphereCollisionRadius=2
      boundBone {boneName=dyn_hair_midr_02}
      boxExtents=<12 20 20>
      constraintSetup {angularLimitsMax=<15 10 2>, angularLimitsMin=<-15 -10 -45>, angularTarget=<1 0 0>}
      localJointOffset=<2 0 0>
     }]
    relativeSpaceBone {boneName=root}
    sphericalLimits [
     {limitRadius=12, drivingBone {boneName=spine_04}, sphereLocalOffset=<2 7 0>}
     {limitRadius=13.5, drivingBone {boneName=spine_03}, sphereLocalOffset=<2 7 0>}
     {limitRadius=12, drivingBone {boneName=spine_05}, sphereLocalOffset=<3 7.5 0>}
    ]
   }
   n14 = Anim Dynamics - Bone: dyn_hair_midr_03!(ComponentPose=@n13, Alpha=0.5, GravityOverride:Vector="0,0,-980", ExternalForce="0,0,0") -> (Pose) {
    angularDampingOverride=0.9999
    angularSpringConstant=75
    bAngularSpring=true
    bOverrideAngularDamping=true
    bUseGravityOverride=true
    bUseSphericalLimits=true
    simulationSpace=RootRelative
    alphaBoolBlend {blendOption=Linear}
    boundBone {boneName=dyn_hair_midr_03}
    componentAppliedLinearAccClamp=<3500 3500 3500>
    componentLinearAccScale=<1 1 1>
    componentLinearVelScale=<1 1 1>
    gravityOverride {z=-980}
    physicsBodyDefinitions [{
      collisionType=CustomSphere
      sphereCollisionRadius=2
      boundBone {boneName=dyn_hair_midr_03}
      boxExtents=<12 20 20>
      constraintSetup {angularLimitsMax=<15 10 2>, angularLimitsMin=<-15 -10 -45>, angularTarget=<1 0 0>}
      localJointOffset=<2 0 0>
     }]
    relativeSpaceBone {boneName=root}
    sphericalLimits [
     {limitRadius=12, drivingBone {boneName=spine_04}, sphereLocalOffset=<2 7 0>}
     {limitRadius=13.5, drivingBone {boneName=spine_03}, sphereLocalOffset=<2 7 0>}
     {limitRadius=12, drivingBone {boneName=spine_05}, sphereLocalOffset=<3 7.5 0>}
    ]
   }
   n15 = Anim Dynamics - Bone: dyn_hair_midr_04!(ComponentPose=@n14, Alpha=0.8, GravityOverride:Vector="0,0,-980", ExternalForce="0,0,0") -> (Pose) {
    angularDampingOverride=0.9999
    angularSpringConstant=75
    bAngularSpring=true
    bOverrideAngularDamping=true
    bUseGravityOverride=true
    bUseSphericalLimits=true
    simulationSpace=RootRelative
    alphaBoolBlend {blendOption=Linear}
    boundBone {boneName=dyn_hair_midr_04}
    componentAppliedLinearAccClamp=<3500 3500 3500>
    componentLinearAccScale=<1 1 1>
    componentLinearVelScale=<1 1 1>
    gravityOverride {z=-980}
    physicsBodyDefinitions [{
      collisionType=CustomSphere
      sphereCollisionRadius=2
      boundBone {boneName=dyn_hair_midr_04}
      boxExtents=<12 20 20>
      constraintSetup {angularLimitsMax=<15 10 2>, angularLimitsMin=<-15 -10 -45>, angularTarget=<1 0 0>}
      localJointOffset=<2 0 0>
     }]
    relativeSpaceBone {boneName=root}
    sphericalLimits [
     {limitRadius=12, drivingBone {boneName=spine_04}, sphereLocalOffset=<2 7 0>}
     {limitRadius=13.5, drivingBone {boneName=spine_03}, sphereLocalOffset=<2 7 0>}
     {limitRadius=12, drivingBone {boneName=spine_05}, sphereLocalOffset=<3 7.5 0>}
    ]
   }
   n16 = Anim Dynamics - Bone: dyn_hair_l_02!(ComponentPose=@n15, Alpha=1, GravityOverride:Vector="0,0,-980", ExternalForce="0,0,0") -> (Pose) {
    angularDampingOverride=0.9999
    angularSpringConstant=75
    bAngularSpring=true
    bOverrideAngularDamping=true
    bUseGravityOverride=true
    bUseSphericalLimits=true
    simulationSpace=RootRelative
    alphaBoolBlend {blendOption=Linear}
    boundBone {boneName=dyn_hair_l_02}
    componentAppliedLinearAccClamp=<3500 3500 3500>
    componentLinearAccScale=<1 1 1>
    componentLinearVelScale=<1 1 1>
    gravityOverride {z=-980}
    physicsBodyDefinitions [{
      collisionType=CustomSphere
      sphereCollisionRadius=3
      boundBone {boneName=dyn_hair_l_02}
      boxExtents=<12 20 20>
      constraintSetup {angularLimitsMax=<15 10 10>, angularLimitsMin=<-15 -10 -45>, angularTarget=<1 0 0>}
      localJointOffset=<2 0 0>
     }]
    relativeSpaceBone {boneName=root}
   }
   n17 = Anim Dynamics - Bone: dyn_hair_l_03!(ComponentPose=@n16, Alpha=1, GravityOverride:Vector="0,0,-980", ExternalForce="0,0,0") -> (Pose) {
    angularDampingOverride=0.9999
    angularSpringConstant=75
    bAngularSpring=true
    bOverrideAngularDamping=true
    bUseGravityOverride=true
    bUseSphericalLimits=true
    simulationSpace=RootRelative
    alphaBoolBlend {blendOption=Linear}
    boundBone {boneName=dyn_hair_l_03}
    componentAppliedLinearAccClamp=<3500 3500 3500>
    componentLinearAccScale=<1 1 1>
    componentLinearVelScale=<1 1 1>
    gravityOverride {z=-980}
    physicsBodyDefinitions [{
      collisionType=CustomSphere
      sphereCollisionRadius=3
      boundBone {boneName=dyn_hair_l_03}
      boxExtents=<12 20 20>
      constraintSetup {angularLimitsMax=<15 10 5>, angularLimitsMin=<-15 -10 -45>, angularTarget=<1 0 0>}
      localJointOffset=<2 0 0>
     }]
    relativeSpaceBone {boneName=root}
    sphericalLimits [{limitRadius=4, drivingBone {boneName=clavicle_l}, sphereLocalOffset=<6 0 3>}]
   }
   n18 = Anim Dynamics - Bone: dyn_hair_l_04!(ComponentPose=@n17, Alpha=1, GravityOverride:Vector="0,0,-980", ExternalForce="0,0,0") -> (Pose) {
    angularDampingOverride=0.9999
    angularSpringConstant=75
    bAngularSpring=true
    bOverrideAngularDamping=true
    bUseGravityOverride=true
    bUseSphericalLimits=true
    simulationSpace=RootRelative
    alphaBoolBlend {blendOption=Linear}
    boundBone {boneName=dyn_hair_l_04}
    componentAppliedLinearAccClamp=<3500 3500 3500>
    componentLinearAccScale=<1 1 1>
    componentLinearVelScale=<1 1 1>
    gravityOverride {z=-980}
    physicsBodyDefinitions [{
      collisionType=CustomSphere
      sphereCollisionRadius=3
      boundBone {boneName=dyn_hair_l_04}
      boxExtents=<12 20 20>
      constraintSetup {angularLimitsMax=<15 10 2>, angularLimitsMin=<-15 -10 -45>, angularTarget=<1 0 0>}
      localJointOffset=<2 0 0>
     }]
    relativeSpaceBone {boneName=root}
    sphericalLimits [{limitRadius=4, drivingBone {boneName=clavicle_l}, sphereLocalOffset=<6 0 3>}]
   }
   n19 = Anim Dynamics - Bone: dyn_hair_r_02!(ComponentPose=@n18, Alpha=0.5, GravityOverride:Vector="0,0,-980", ExternalForce="0,0,0") -> (Pose) {
    angularDampingOverride=0.9999
    angularSpringConstant=75
    bAngularSpring=true
    bOverrideAngularDamping=true
    bUseGravityOverride=true
    simulationSpace=RootRelative
    alphaBoolBlend {blendOption=Linear}
    boundBone {boneName=dyn_hair_r_02}
    componentAppliedLinearAccClamp=<3500 3500 3500>
    componentLinearAccScale=<1.2 1.2 1.2>
    componentLinearVelScale=<1.2 1.2 1.2>
    gravityOverride {z=-980}
    physicsBodyDefinitions [{
      collisionType=CustomSphere
      sphereCollisionRadius=2
      boundBone {boneName=dyn_hair_r_02}
      boxExtents=<10 20 20>
      constraintSetup {angularLimitsMax=<1 5 15>, angularLimitsMin=<-1 -30 -15>, angularTarget=<0 0 0>}
      localJointOffset=<2 0 0>
     }]
    relativeSpaceBone {boneName=root}
   }
   n20 = Anim Dynamics - Bone: dyn_hair_r_03!(ComponentPose=@n19, Alpha=0.8, GravityOverride:Vector="0,0,-980", ExternalForce="0,0,0") -> (Pose) {
    angularDampingOverride=0.9999
    angularSpringConstant=75
    bAngularSpring=true
    bOverrideAngularDamping=true
    bUseGravityOverride=true
    simulationSpace=RootRelative
    alphaBoolBlend {blendOption=Linear}
    boundBone {boneName=dyn_hair_r_03}
    componentAppliedLinearAccClamp=<3500 3500 3500>
    componentLinearAccScale=<1.2 1.2 1.2>
    componentLinearVelScale=<1.2 1.2 1.2>
    gravityOverride {z=-980}
    physicsBodyDefinitions [{
      collisionType=CustomSphere
      sphereCollisionRadius=2
      boundBone {boneName=dyn_hair_r_03}
      boxExtents=<10 20 20>
      constraintSetup {angularLimitsMax=<1 5 15>, angularLimitsMin=<-1 -30 -15>, angularTarget=<0 0 0>}
      localJointOffset=<2 0 0>
     }]
    relativeSpaceBone {boneName=root}
   }
   n21 = Anim Dynamics - Bone: dyn_hair_r_04!(ComponentPose=@n20, Alpha=0.5, GravityOverride:Vector="0,0,-980", ExternalForce="0,0,0") -> (Pose) {
    angularDampingOverride=0.9999
    angularSpringConstant=75
    bAngularSpring=true
    bOverrideAngularDamping=true
    bUseGravityOverride=true
    simulationSpace=RootRelative
    alphaBoolBlend {blendOption=Linear}
    boundBone {boneName=dyn_hair_r_04}
    componentAppliedLinearAccClamp=<3500 3500 3500>
    componentLinearAccScale=<1.2 1.2 1.2>
    componentLinearVelScale=<1.2 1.2 1.2>
    gravityOverride {z=-980}
    physicsBodyDefinitions [{
      collisionType=CustomSphere
      sphereCollisionRadius=2
      boundBone {boneName=dyn_hair_r_04}
      boxExtents=<10 20 20>
      constraintSetup {angularLimitsMax=<1 5 15>, angularLimitsMin=<-1 -30 -15>, angularTarget=<0 0 0>}
      localJointOffset=<2 0 0>
     }]
    relativeSpaceBone {boneName=root}
   }
   n22 = Anim Dynamics - Bone: dyn_hair_fr_r_02!(ComponentPose=@n21, Alpha=1, GravityOverride:Vector="0,0,-980", ExternalForce="0,0,0") -> (Pose) {
    angularDampingOverride=0.9999
    angularSpringConstant=75
    bAngularSpring=true
    bOverrideAngularDamping=true
    bUseGravityOverride=true
    simulationSpace=RootRelative
    alphaBoolBlend {blendOption=Linear}
    boundBone {boneName=dyn_hair_fr_r_02}
    componentAppliedLinearAccClamp=<3500 3500 3500>
    componentLinearAccScale=<1.2 1.2 1.2>
    componentLinearVelScale=<1.2 1.2 1.2>
    gravityOverride {z=-980}
    physicsBodyDefinitions [{
      collisionType=CustomSphere
      sphereCollisionRadius=2
      boundBone {boneName=dyn_hair_fr_r_02}
      boxExtents=<10 20 20>
      constraintSetup {angularLimitsMax=<1 5 15>, angularLimitsMin=<-1 -30 -15>, angularTarget=<0 0 0>}
      localJointOffset=<2 0 0>
     }]
    relativeSpaceBone {boneName=root}
   }
   n23 = Anim Dynamics - Bone: dyn_hair_fr_r_03!(ComponentPose=@n22, Alpha=1, GravityOverride:Vector="0,0,-980", ExternalForce="0,0,0") -> (Pose) {
    angularDampingOverride=0.9999
    angularSpringConstant=75
    bAngularSpring=true
    bOverrideAngularDamping=true
    bUseGravityOverride=true
    simulationSpace=RootRelative
    alphaBoolBlend {blendOption=Linear}
    boundBone {boneName=dyn_hair_fr_r_03}
    componentAppliedLinearAccClamp=<3500 3500 3500>
    componentLinearAccScale=<1.2 1.2 1.2>
    componentLinearVelScale=<1.2 1.2 1.2>
    gravityOverride {z=-980}
    physicsBodyDefinitions [{
      collisionType=CustomSphere
      sphereCollisionRadius=2
      boundBone {boneName=dyn_hair_fr_r_03}
      boxExtents=<10 20 20>
      constraintSetup {angularLimitsMax=<1 5 15>, angularLimitsMin=<-1 -30 -15>, angularTarget=<0 0 0>}
      localJointOffset=<2 0 0>
     }]
    relativeSpaceBone {boneName=root}
   }
   n24 = Anim Dynamics - Bone: dyn_hair_fr_r_04!(ComponentPose=@n23, Alpha=1, GravityOverride:Vector="0,0,-980", ExternalForce="0,0,0") -> (Pose) {
    angularDampingOverride=0.9999
    angularSpringConstant=75
    bAngularSpring=true
    bOverrideAngularDamping=true
    bUseGravityOverride=true
    simulationSpace=RootRelative
    alphaBoolBlend {blendOption=Linear}
    boundBone {boneName=dyn_hair_fr_r_04}
    componentAppliedLinearAccClamp=<3500 3500 3500>
    componentLinearAccScale=<1.2 1.2 1.2>
    componentLinearVelScale=<1.2 1.2 1.2>
    gravityOverride {z=-980}
    physicsBodyDefinitions [{
      collisionType=CustomSphere
      sphereCollisionRadius=2
      boundBone {boneName=dyn_hair_fr_r_04}
      boxExtents=<10 20 20>
      constraintSetup {angularLimitsMax=<1 5 15>, angularLimitsMin=<-1 -30 -15>, angularTarget=<0 0 0>}
      localJointOffset=<2 0 0>
     }]
    relativeSpaceBone {boneName=root}
   }
   n25 = Anim Dynamics - Bone: dyn_hair_fr_r_05!(ComponentPose=@n24, Alpha=1, GravityOverride:Vector="0,0,-980", ExternalForce="0,0,0") -> (Pose) {
    angularDampingOverride=0.9999
    angularSpringConstant=75
    bAngularSpring=true
    bOverrideAngularDamping=true
    bUseGravityOverride=true
    simulationSpace=RootRelative
    alphaBoolBlend {blendOption=Linear}
    boundBone {boneName=dyn_hair_fr_r_05}
    componentAppliedLinearAccClamp=<3500 3500 3500>
    componentLinearAccScale=<1.2 1.2 1.2>
    componentLinearVelScale=<1.2 1.2 1.2>
    gravityOverride {z=-980}
    physicsBodyDefinitions [{
      collisionType=CustomSphere
      sphereCollisionRadius=2
      boundBone {boneName=dyn_hair_fr_r_05}
      boxExtents=<10 20 20>
      constraintSetup {angularLimitsMax=<1 5 15>, angularLimitsMin=<-1 -30 -15>, angularTarget=<0 0 0>}
      localJointOffset=<2 0 0>
     }]
    relativeSpaceBone {boneName=root}
   }
   n26 = Anim Dynamics - Bone: dyn_hair_bk_l_01!(ComponentPose=@n25, Alpha=0.8, GravityOverride:Vector="0,0,-980", ExternalForce="0,0,0") -> (Pose) {
    angularDampingOverride=0.98
    angularSpringConstant=150
    bAngularSpring=true
    bOverrideAngularDamping=true
    bUseGravityOverride=true
    simulationSpace=RootRelative
    alphaBoolBlend {blendOption=Linear}
    boundBone {boneName=dyn_hair_bk_l_01}
    componentAppliedLinearAccClamp=<1500 1500 1500>
    componentLinearAccScale=<0.6 0.6 0.6>
    componentLinearVelScale=<0.6 0.6 0.6>
    gravityOverride {z=-980}
    physicsBodyDefinitions [{
      sphereCollisionRadius=0
      boundBone {boneName=dyn_hair_bk_l_01}
      boxExtents=<12 15 15>
      constraintSetup {angularLimitsMax=<1 5 0>, angularLimitsMin=<-1 -5 -15>, angularTarget=<1 0 0>}
      localJointOffset=<1 0 0>
     }]
    relativeSpaceBone {boneName=root}
   }
   n27 = Anim Dynamics - Bone: dyn_hair_bk_01!(ComponentPose=@n26, Alpha=0.8, GravityOverride:Vector="0,0,-980", ExternalForce="0,0,0") -> (Pose) {
    angularDampingOverride=0.98
    angularSpringConstant=150
    bAngularSpring=true
    bOverrideAngularDamping=true
    bUseGravityOverride=true
    simulationSpace=RootRelative
    alphaBoolBlend {blendOption=Linear}
    boundBone {boneName=dyn_hair_bk_01}
    componentAppliedLinearAccClamp=<1500 1500 1500>
    componentLinearAccScale=<0.3 0.4 0.6>
    componentLinearVelScale=<0.3 0.4 0.6>
    gravityOverride {z=-980}
    physicsBodyDefinitions [{
      sphereCollisionRadius=0
      boundBone {boneName=dyn_hair_bk_01}
      boxExtents=<12 15 15>
      constraintSetup {angularLimitsMax=<1 5 10>, angularLimitsMin=<-1 -20 -10>, angularTarget=<1 0 0>}
      localJointOffset=<1 0 0>
     }]
    relativeSpaceBone {boneName=root}
   }
   n28 = Anim Dynamics - Bone: dyn_hair_bk_r_01!(ComponentPose=@n27, Alpha=0.8, GravityOverride:Vector="0,0,-980", ExternalForce="0,0,0") -> (Pose) {
    angularDampingOverride=0.98
    angularSpringConstant=150
    bAngularSpring=true
    bOverrideAngularDamping=true
    bUseGravityOverride=true
    simulationSpace=RootRelative
    alphaBoolBlend {blendOption=Linear}
    boundBone {boneName=dyn_hair_bk_r_01}
    componentAppliedLinearAccClamp=<1500 1500 1500>
    componentLinearAccScale=<0.3 0.4 0.6>
    componentLinearVelScale=<0.3 0.4 0.6>
    gravityOverride {z=-980}
    physicsBodyDefinitions [{
      sphereCollisionRadius=0
      boundBone {boneName=dyn_hair_bk_r_01}
      boxExtents=<12 15 15>
      constraintSetup {angularLimitsMax=<1 5 10>, angularLimitsMin=<-1 -20 -15>, angularTarget=<1 0 0>}
      localJointOffset=<1 0 0>
     }]
    relativeSpaceBone {boneName=root}
   }
   n29 = Look At - Bone: dyn_hair_fr_l_02(ComponentPose=@n28, Alpha=0.3) {
    alphaBoolBlend {blendOption=Linear}
    boneToModify {boneName=dyn_hair_fr_l_02}
    lookAtLocation=<-61.769 -85.1517 -37.7435>
    lookAtTarget {boneReference {boneName=head}}
    lookAt_Axis {axis=<1 0>}
   }
   n30 = Anim Dynamics - Bone: dyn_hair_fr_l_01!(ComponentPose=@n29, Alpha=0.8, GravityOverride:Vector="0,0,-980", ExternalForce="0,0,0") -> (Pose) {
    angularDampingOverride=0.9999
    angularSpringConstant=500
    bAngularSpring=true
    bLinearSpring=true
    bOverrideAngularDamping=true
    bUseGravityOverride=true
    bUseSphericalLimits=true
    linearSpringConstant=50
    simulationSpace=BoneRelative
    alphaBoolBlend {blendOption=Linear}
    boundBone {boneName=dyn_hair_fr_l_01}
    componentAppliedLinearAccClamp=<1500 1500 0>
    componentLinearAccScale=<0.3 0.3 0.3>
    componentLinearVelScale=<0.3 0.3 0.5>
    gravityOverride {z=-980}
    physicsBodyDefinitions [{
      sphereCollisionRadius=0
      boundBone {boneName=dyn_hair_fr_l_01}
      boxExtents=<5 20 20>
      constraintSetup {angularLimitsMax=<0.5 4 20>, angularLimitsMin=<-0.5 -10 -1>, angularTarget=<0 0 0>}
      localJointOffset=<3 0 0>
     }]
    relativeSpaceBone {boneName=root}
    sphericalLimits [{limitRadius=5, drivingBone {boneName=head}, sphereLocalOffset=<0 3 0>}]
   }
   n31 = Anim Dynamics - Bone: dyn_hair_fr_l_02!(ComponentPose=@n30, Alpha=1, GravityOverride:Vector="0,0,-980", ExternalForce="0,0,0") -> (Pose) {
    angularDampingOverride=0.9999
    angularSpringConstant=500
    bAngularSpring=true
    bLinearSpring=true
    bOverrideAngularDamping=true
    bUseGravityOverride=true
    bUseSphericalLimits=true
    linearSpringConstant=50
    simulationSpace=BoneRelative
    alphaBoolBlend {blendOption=Linear}
    boundBone {boneName=dyn_hair_fr_l_02}
    componentAppliedLinearAccClamp=<1500 1500 0>
    componentLinearAccScale=<0.3 0.3 0.3>
    componentLinearVelScale=<0.3 0.3 0.5>
    gravityOverride {z=-980}
    physicsBodyDefinitions [{
      sphereCollisionRadius=0
      boundBone {boneName=dyn_hair_fr_l_02}
      boxExtents=<5 20 20>
      constraintSetup {angularLimitsMax=<0.5 4 20>, angularLimitsMin=<-0.5 -10 -1>, angularTarget=<0 0 0>}
      localJointOffset=<3 0 0>
     }]
    relativeSpaceBone {boneName=root}
    sphericalLimits [{limitRadius=5, drivingBone {boneName=head}, sphereLocalOffset=<0 3 0>}]
   }
   n32 = Anim Dynamics - Bone: dyn_hair_fr_l_03!(ComponentPose=@n31, Alpha=1, GravityOverride:Vector="0,0,-980", ExternalForce="0,0,0") -> (Pose) {
    angularDampingOverride=0.9999
    angularSpringConstant=500
    bAngularSpring=true
    bLinearSpring=true
    bOverrideAngularDamping=true
    bUseGravityOverride=true
    bUseSphericalLimits=true
    linearSpringConstant=50
    simulationSpace=BoneRelative
    alphaBoolBlend {blendOption=Linear}
    boundBone {boneName=dyn_hair_fr_l_03}
    componentAppliedLinearAccClamp=<1500 1500 0>
    componentLinearAccScale=<0.3 0.3 0.3>
    componentLinearVelScale=<0.3 0.3 0.5>
    gravityOverride {z=-980}
    physicsBodyDefinitions [{
      sphereCollisionRadius=0
      boundBone {boneName=dyn_hair_fr_l_03}
      boxExtents=<5 20 20>
      constraintSetup {angularLimitsMax=<0.5 4 30>, angularLimitsMin=<-0.5 -8 -8>, angularTarget=<0 0 0>}
      localJointOffset=<3 0 0>
     }]
    relativeSpaceBone {boneName=root}
    sphericalLimits [{limitRadius=5, drivingBone {boneName=head}, sphereLocalOffset=<0 3 0>}]
   }
   n33 = Anim Dynamics - Bone: dyn_hair_fr_l_04!(ComponentPose=@n32, Alpha=1, GravityOverride:Vector="0,0,-980", ExternalForce="0,0,0") -> (Pose) {
    angularDampingOverride=0.9999
    angularSpringConstant=500
    bAngularSpring=true
    bLinearSpring=true
    bOverrideAngularDamping=true
    bUseGravityOverride=true
    bUseSphericalLimits=true
    linearSpringConstant=50
    simulationSpace=BoneRelative
    alphaBoolBlend {blendOption=Linear}
    boundBone {boneName=dyn_hair_fr_l_04}
    componentAppliedLinearAccClamp=<1500 1500 0>
    componentLinearAccScale=<0.3 0.3 0.3>
    componentLinearVelScale=<0.3 0.3 0.5>
    gravityOverride {z=-980}
    physicsBodyDefinitions [{
      sphereCollisionRadius=0
      boundBone {boneName=dyn_hair_fr_l_04}
      boxExtents=<5 20 20>
      constraintSetup {angularLimitsMax=<0.5 4 30>, angularLimitsMin=<-0.5 -8 -8>, angularTarget=<0 0 0>}
      localJointOffset=<3 0 0>
     }]
    relativeSpaceBone {boneName=root}
    sphericalLimits [{limitRadius=5, drivingBone {boneName=head}, sphereLocalOffset=<0 3 0>}]
   }
   n34 = Anim Dynamics - Bone: dyn_hair_fr_l_05!(ComponentPose=@n33, Alpha=1, GravityOverride:Vector="0,0,-980", ExternalForce="0,0,0") -> (Pose) {
    angularDampingOverride=0.9999
    angularSpringConstant=500
    bAngularSpring=true
    bLinearSpring=true
    bOverrideAngularDamping=true
    bUseGravityOverride=true
    bUseSphericalLimits=true
    linearSpringConstant=50
    simulationSpace=BoneRelative
    alphaBoolBlend {blendOption=Linear}
    boundBone {boneName=dyn_hair_fr_l_05}
    componentAppliedLinearAccClamp=<1500 1500 0>
    componentLinearAccScale=<0.3 0.3 0.3>
    componentLinearVelScale=<0.3 0.3 0.5>
    gravityOverride {z=-980}
    physicsBodyDefinitions [{
      collisionType=CustomSphere
      sphereCollisionRadius=2.5
      boundBone {boneName=dyn_hair_fr_l_05}
      boxExtents=<5 20 20>
      constraintSetup {angularLimitsMax=<1 3 30>, angularLimitsMin=<-1 -8 -8>, angularTarget=<0 0 0>}
      localJointOffset=<3 0 0>
     }]
    relativeSpaceBone {boneName=root}
    sphericalLimits [
     {limitRadius=5, drivingBone {boneName=head}, sphereLocalOffset=<-6 3 0>}
     {limitRadius=5, drivingBone {boneName=clavicle_l}, sphereLocalOffset=<6 0 0>}
     {limitRadius=6, drivingBone {boneName=neck_01}, sphereLocalOffset=<-3 3 0>}
    ]
   }
   n35 = Anim Dynamics - Bone: dyn_hair_fr_l_06!(ComponentPose=@n34, Alpha=1, GravityOverride:Vector="0,0,-980", ExternalForce="0,0,0") -> (Pose) {
    angularDampingOverride=0.9999
    angularSpringConstant=500
    bAngularSpring=true
    bLinearSpring=true
    bOverrideAngularDamping=true
    bUseGravityOverride=true
    bUseSphericalLimits=true
    linearSpringConstant=50
    simulationSpace=BoneRelative
    alphaBoolBlend {blendOption=Linear}
    boundBone {boneName=dyn_hair_fr_l_06}
    componentAppliedLinearAccClamp=<1500 1500 0>
    componentLinearAccScale=<0.3 0.3 0.3>
    componentLinearVelScale=<0.3 0.3 0.5>
    gravityOverride {z=-980}
    physicsBodyDefinitions [{
      collisionType=CustomSphere
      sphereCollisionRadius=2.5
      boundBone {boneName=dyn_hair_fr_l_06}
      boxExtents=<5 20 20>
      constraintSetup {angularLimitsMax=<1 3 30>, angularLimitsMin=<-1 -8 -8>, angularTarget=<0 0 0>}
      localJointOffset=<3 0 0>
     }]
    relativeSpaceBone {boneName=root}
    sphericalLimits [
     {limitRadius=5, drivingBone {boneName=head}, sphereLocalOffset=<-6 3 0>}
     {limitRadius=5, drivingBone {boneName=clavicle_l}, sphereLocalOffset=<6 1 -2>}
     {limitRadius=4, drivingBone {boneName=neck_01}, sphereLocalOffset=<-3 3 0>}
     {limitRadius=10, drivingBone {boneName=spine_05}, sphereLocalOffset=<0 0 -3>}
    ]
   }
   n36 = Anim Dynamics - Bone: dyn_hair_fr_l_07!(ComponentPose=@n35, Alpha=1, GravityOverride:Vector="0,0,-980", ExternalForce="0,0,0") -> (Pose) {
    angularDampingOverride=0.9999
    angularSpringConstant=500
    bAngularSpring=true
    bLinearSpring=true
    bOverrideAngularDamping=true
    bUseGravityOverride=true
    bUseSphericalLimits=true
    linearSpringConstant=50
    simulationSpace=BoneRelative
    alphaBoolBlend {blendOption=Linear}
    boundBone {boneName=dyn_hair_fr_l_07}
    componentAppliedLinearAccClamp=<1500 1500 0>
    componentLinearAccScale=<0.3 0.3 0.3>
    componentLinearVelScale=<0.3 0.3 0.5>
    gravityOverride {z=-980}
    physicsBodyDefinitions [{
      sphereCollisionRadius=1
      boundBone {boneName=dyn_hair_fr_l_07}
      boxExtents=<5 20 20>
      constraintSetup {angularLimitsMax=<1 3 30>, angularLimitsMin=<-1 -8 -8>, angularTarget=<0 0 0>}
      localJointOffset=<3 0 0>
     }]
    relativeSpaceBone {boneName=root}
    sphericalLimits [
     {limitRadius=5, drivingBone {boneName=head}, sphereLocalOffset=<-8 3 0>}
     {limitRadius=4, drivingBone {boneName=clavicle_l}, sphereLocalOffset=<4 0 0>}
     {limitRadius=4, drivingBone {boneName=neck_01}, sphereLocalOffset=<-3 3 0>}
     {limitRadius=9, drivingBone {boneName=spine_05}, sphereLocalOffset=<0 0 -3>}
    ]
   }
   n37 = Control Rig!(
    Source=Component To Local(ComponentPose=@n36)
    Alpha
    Melee:double=Get MeleeSwing() -> (MeleeSwing:double)
    Crouch:bool=Get bIsCrouching()
    Skydive:bool=Get bIsSkyDiving()
    Emote:double=Get Emote() -> (Emote:double)
    Sprint:double=Get Sprint() -> (Sprint:double)
    ADS_Rifle:bool=Get isAimingRifle() -> (isAimingRifle:bool)
    ADS_Pistol:bool=Get isAiming_Pistol() -> (isAiming_Pistol:bool)
    Moving:double=Get Moving() -> (Moving:double)
    Sailor:double=Get Sailor() -> (Sailor:double)
    Rifle:double=Get Rifle() -> (Rifle:double)
    Blazer:double=Get Blazer() -> (Blazer:double)
   ) -> (Pose) {
    alphaBoolBlend {blendOption=Linear}
    controlRigAssetReference {blueprintRigClass=&/F_MED_Pencil_Pear_Hair_CtrlRig.F_MED_Pencil_Pear_Hair_CtrlRig_C}
   }
   Save cached pose 'Pose'(Pose=@n37) {cacheName=Pose}
  ]}
 EventGraph {nodes [
   Event BlueprintUpdateAnimation()
   > n1 = Sequence!(execute) -> (then_0, then_1, then_2:Exec, then_3:Exec, then_4:Exec)
   > n2 = Gravity Override State Function DEPRECATED(CharacterPartAnimInstance=Self-Reference(), GameTimeInSeconds=GetGameTimeInSeconds(), GravityOverrideParams=Get Gravity Override Params())
   > Set Gravity Override(Gravity Override=@n2)
   Set MeleeSwing(execute=@n1.then_1, MeleeSwing:double=SelectFloat(A=1, B=0, bPickA=Get bIsPlayingMeleeAnim())) -> (Output_Get:double)
   n3 = BreakTransform(InTransform=GetDeltaTransformFromRefPose(self=GetOwnerSkeletalMeshComponent(), BoneName=head, BaseName=spine_05))
   > Set Head Up(Head Up:double=MapRangeClamped(Value=BreakRotator(InRot=@n3.Rotation).Yaw, InRangeA=-3, InRangeB=-40)) -> (Output_Get:double)
   Set CrouchADS(execute=@n1.then_2, CrouchADS:bool=AND Boolean(A=Get bIsCrouching(), B=Get bIsTargeting())) -> (Output_Get:bool)
   > Set Emote(Emote:double=SelectFloat(A=1, B=0, bPickA=Get bIsPlayingEmote())) -> (Output_Get:double)
   n4 = Get bIsFalling()
   n5 = Get bIsJumping()
   > Set Sprint(Sprint:double=SelectFloat(A=1, B=0, bPickA=OR Boolean(A=@n4, B=@n5))) -> (Output_Get:double)
   > Set Is Wearing Backbling(Is Wearing Backbling=CheckForBackbling())
   > Set Moving(Moving:double=SelectFloat(A=1, B=0, bPickA=Get bIsMoving2D())) -> (Output_Get:double)
   n6 = CharacterItemDefinitionHasAnyMetaTag(Tags=Make Array("[0]"=(TagName="Cosmetics.MetaTags.Pencil.Upper.Sailor")))
   > "Set Sailor_Jacket "("Sailor_Jacket ":bool=@n6) -> (Output_Get:bool)
   > Set Sailor(Sailor:double=SelectFloat(A=1, B=0, bPickA=@n6)) -> (Output_Get:double)
   n7 = Make Array("[0]"=(TagName="Cosmetics.MetaTags.Pencil.Upper.BlazerCuff"))
   > Set Blazer(Blazer:double=SelectFloat(A=1, B=0, bPickA=CharacterItemDefinitionHasAnyMetaTag(Tags=@n7))) -> (Output_Get:double)
   > Set skydive_off(skydive_off:double=SelectFloat(A=0, B=1, bPickA=Get bIsSkyDiving())) -> (Output_Get:double)
   n8 = Cast To FortPlayerAnimInstance(execute=@n1.then_3, Object=GetAnimInstance(self=GetOwnerSkeletalMeshComponent()))
   > Set Pistol(Pistol:double=0) -> (Output_Get:double)
   > Set Rifle(Rifle:double=0) -> (Output_Get:double)
   > n9 = Switch on EFortWeaponCoreAnimation!(execute, Selection:EFortWeaponCoreAnimation Enum=Get WeaponCoreAnim(self=@n8)) -> (
    Melee:Exec
    Pistol:Exec
    Shotgun:Exec
    PaperBlueprint:Exec
    Rifle:Exec
    MeleeOneHand:Exec
    MachinePistol:Exec
    RocketLauncher:Exec
    GrenadeLauncher:Exec
    GoingCommando:Exec
    AssaultRifle:Exec
    TacticalShotgun:Exec
    SniperRifle:Exec
    TrapPlacement:Exec
    ShoulderLauncher:Exec
    AbilityDecoTool:Exec
    Crossbow:Exec
    C4:Exec
    RemoteControl:Exec
    DualWield:Exec
    AR_BullPup:Exec
    AR_ForwardGrip:Exec
    MedPackPaddles:Exec
    SMG_P90:Exec
    AR_DrumGun:Exec
    Consumable_Small:Exec
    Consumable_Large:Exec
    Balloon:Exec
    MountedTurret:Exec
    CreativeTool:Exec
    ExplosiveBow:Exec
    AshtonIndigo:Exec
    AshtonChicago:Exec
    MeleeDualWield:Exec
    Unarmed:Exec
   )
   Branch(execute=[@n9.AssaultRifle, @n9.Crossbow, @n9.GrenadeLauncher, @n9.Rifle, @n9.RocketLauncher, @n9.Shotgun, @n9.SniperRifle, @n9.TacticalShotgun])
   > Set Rifle(Rifle=1)
   > n10 = Branch(Condition=Get bIsTargeting())
   > Set isAimingRifle(isAimingRifle:bool=true) -> (Output_Get:bool)
   Set isAimingRifle(execute=@n10.else, isAimingRifle=false)
   Branch(execute=[@n9.MachinePistol, @n9.Pistol])
   > Set Pistol(Pistol=1)
   > n11 = Branch(Condition=Get bIsTargeting())
   > Set isAiming_Pistol(isAiming_Pistol:bool=true) -> (Output_Get:bool)
   Set isAiming_Pistol(execute=@n11.else, isAiming_Pistol=false)
   TryGetPawnOwner()
  ]}
]
variables [
 Blazer {type=number}
 CrouchADS {type=boolean}
 Delta Time X {type=number}
 Emote {type=number}
 Gravity Override {type=Vector}
 Gravity Override Params {type=GravityOverrideParamsStruct}
 Head Up {type=number}
 Is Wearing Backbling {type=boolean}
 isAiming_Pistol {type=boolean}
 isAimingRifle {type=boolean}
 MeleeSwing {type=number}
 Moving {type=number}
 Pistol {type=number}
 Rifle {type=number}
 Sailor {type=number}
 "Sailor_Jacket " {type=boolean}
 skydive_off {type=number}
 Sprint {type=number}
]
